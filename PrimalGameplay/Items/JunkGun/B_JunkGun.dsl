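; Junk Gun weapon Blueprint, handling motor idle audio, a low-ammo sound, ammo-pickup notifications and suction state.
; Equipping binds vehicle, emote and holster events that drive a fading motor loop, with separate first- and third-person sounds.
; StoreSuctionValues passes server suction hit and resource data to owning clients, cached for a getter.
; Ammo pickup broadcasts cosmetically through the global event router; unequip unbinds OnUnholstered, although equip binds OnHolstered.
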(event Custom|On%20Ammo%20Pickup
  ;[1 1 1] Ammo pickup dispatcher for catching inside suction GCNL
  (Networking|CanExecuteCosmeticEvents
    (:True
      (GameplayEventRouter|BroadcastEvent (GameplayEventRouter|GetGlobalEventRouter) (Utilities|Struct|MakeGameplayEventonJunkGunPickupAmmo true) self)
      (Default|CallOnAmmoPickup) ; Unsure if this is still being used anywhere, so keeping it just in case
    )))

(event Custom|StoreSuctionValues
  (params (IsHit :type Boolean) (IsNotFull :type Boolean) (HitLocation :type Vector) (StoredResource :type Float%20%28double-precision%29) (Normal :type Vector))
  ;[1 1 1] Replicate suction value from server to owning clients
  (CallFunction|SetSuctionValues IsHit IsNotFull HitLocation StoredResource Normal))

(event EventOnPlayWeaponFireFX
  (params (bPersistentFire :type Boolean) (bSecondaryFire :type Boolean))
  ;[1 1 1] Play low ammo sound when clip is below 1/3 max clip size
  (|CallParentFunction bPersistentFire bSecondaryFire)
  (if (<= (FortWeapon|GetMagazineAmmoCount) (Math|Float|Ceil (* (FortWeapon|GetBulletsPerClip) 0.33)))
    (Audio|PlaySoundatLocation (Variables|WeaponAudio|GetLowAmmoSound) (Transformation|GetActorLocation self))))

(event FortWeapon|EventOnEquippedAndAttached
  ;[1 1 1] Idle Looping Sound
  (|CallParentFunction)
  (bind _asplayer_pawn_athena_generic_parent (Utilities|Casting|CastToPlayerPawn_Athena_Generic_Parent (Game|GetInstigator)))
  (Variables|References|SetWeaponOwnerReference _asplayer_pawn_athena_generic_parent)
  (CallFunction|StartMotorIdleSound)
  (Default|AssignOnEnteredVehicleDispatcher (Variables|References|GetWeaponOwnerReference) (delegate Custom|EnterVehicleEvent))
  (Default|BindEventtoOnExitedVehicleDispatcher (Variables|References|GetWeaponOwnerReference) (delegate Custom|ExitVehicleEvent))
  (EventDispatchers|BindEventtoOnHolsteredEvent (Variables|References|GetWeaponOwnerReference) (EventDispatchers|CreateEvent :!event Custom|OnHolsterHandleMotorIdle))
  (EventDispatchers|BindEventtoOnPawnStartedEmote (Variables|References|GetWeaponOwnerReference) (delegate Custom|EmoteStarted))
  (EventDispatchers|BindEventtoOnPawnStoppedEmote (Variables|References|GetWeaponOwnerReference) (delegate Custom|EmoteStopped)))

(event Custom|EnterVehicleEvent
  ;[1 1 1] Idle Looping Sound
  (CallFunction|StopMotorIdleSound))

(event Custom|ExitVehicleEvent
  ;[1 1 1] Idle Looping Sound
  (CallFunction|StartMotorIdleSound))

(event Custom|EmoteStopped
  (params (MontageItemDef :type Fort%20Item%20Definition%20Object%20Reference) (PawnEmoting :type Fort%20Pawn%20Object%20Reference))
  ;[1 1 1] Idle Looping Sound
  (CallFunction|StartMotorIdleSound))

(event Custom|EmoteStarted
  (params (MontageItemDef :type Fort%20Item%20Definition%20Object%20Reference) (PawnEmoting :type Fort%20Pawn%20Object%20Reference))
  ;[1 1 1] Idle Looping Sound
  (CallFunction|StopMotorIdleSound))

(event Custom|StartMotorIdleSound
  (bind _motor_loop (Variables|Default|GetMotor%20Loop))
  ;[1 1 1] Idle Looping Sound
  (Audio|Components|Audio|SetSound (Variables|Default|GetMotor%20Loop) (select (Pawn|IsLocallyControlled (Variables|References|GetWeaponOwnerReference)) (Variables|WeaponAudio|GetMotorLoop1P) (Variables|WeaponAudio|GetMotorLoop3P)))
  (Utilities|IsValid _motor_loop
    (:Is%20Valid
      (Audio|Components|Audio|FadeIn _motor_loop 1.0 1.0 0.0 "Linear" :!class /Script/Engine.AudioComponent))))

(event Custom|OnHolsterHandleMotorIdle
  (bind _weaponownerreference (Variables|References|GetWeaponOwnerReference))
  ;[1 1 1] Idle Looping Sound
  (bind _vehicle_occupied (Utilities|IsinVehicle _weaponownerreference)
    (:true
      (Utilities|FlowControl|Sequence
        (:then_0
          (bind _returnvalue (GameplayTags|HasAllMatchingGameplayTags(Message) (Utilities|ToObject(Interface) _vehicle_occupied) (Variables|Vehicles|GetPickupTruckTag)))
          (if _returnvalue
            (if (select (>= (Class|FortVehicleInterface|FindSeatIndex _vehicle_occupied _weaponownerreference) 2) true false)
              _)))
        (:then_1
          (bind _returnvalue_2 (GameplayTags|HasAllMatchingGameplayTags(Message) (Utilities|ToObject(Interface) _vehicle_occupied) (Variables|Vehicles|GetHelicopterTag)))
          (if _returnvalue_2
            _)))
      (CallFunction|StartMotorIdleSound))))

(event Custom|StopMotorIdleSound
  (bind _motor_loop (Variables|Default|GetMotor%20Loop))
  ;[1 1 1] Idle Looping Sound
  (Utilities|IsValid _motor_loop
    (:Is%20Valid
      (Audio|Components|Audio|FadeOut _motor_loop 0.15 0.0 "Linear" :!class /Script/Engine.AudioComponent))))

(event FortWeapon|EventOnUnEquippedAndDetached
  (bind _weaponownerreference (Variables|References|GetWeaponOwnerReference))
  ;[1 1 1] Idle Looping Sound
  (|CallParentFunction)
  (CallFunction|StopMotorIdleSound)
  (Default|UnbindEventfromOnEnteredVehicleDispatcher _weaponownerreference (EventDispatchers|CreateEvent :!event Custom|EnterVehicleEvent))
  (Default|UnbindEventfromOnExitedVehicleDispatcher _weaponownerreference (EventDispatchers|CreateEvent :!event Custom|ExitVehicleEvent))
  (EventDispatchers|UnbindEventfromOnUnholsteredEvent _weaponownerreference (EventDispatchers|CreateEvent :!event Custom|OnHolsterHandleMotorIdle))
  (EventDispatchers|UnbindEventfromOnPawnStartedEmote _weaponownerreference (EventDispatchers|CreateEvent :!event Custom|EmoteStarted))
  (EventDispatchers|UnbindEventfromOnPawnStoppedEmote _weaponownerreference (EventDispatchers|CreateEvent :!event Custom|EmoteStopped)))

(fn GetJunkGunSuctionValues
  (return (Variables|StoredValues|GetIsHit) (Variables|StoredValues|GetHitLocation) (Variables|StoredValues|GetIsNotFull) (Variables|StoredValues|GetStoredResource) (Variables|StoredValues|GetNormal)))

(fn SetSuctionValues
  (params (IsHit :type Boolean) (IsNotFull :type Boolean) (HitLocation :type Vector) (StoredResource :type Float%20%28double-precision%29) (Normal :type Vector))
  (Variables|StoredValues|SetIsHit IsHit)
  (Variables|StoredValues|SetIsNotFull IsNotFull)
  (Variables|StoredValues|SetHitLocation HitLocation)
  (Variables|StoredValues|SetStoredResource StoredResource)
  (Variables|StoredValues|SetNormal Normal))

(fn ConstructionScript
  (|CallParentFunction))
