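; Junk Gun projectile that spins its mesh in flight and explodes on stopping, entering water or reaching a fuse timer.
; A second timer delays enabling water interaction after spawn.
; All triggers funnel into TriggerExplode, which detonates on the authority if not already exploded and increments RepEvent.
; The explosion handler kills the projectile and stops looping audio once, while RepEvent's OnRep makes remotes hide the mesh and remove trail FX.
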(event Custom|Explode
  ;[1 1 1] Explode way beyond the range of the weapon
  (CallFunction|TriggerExplode))

(event Custom|EnableWater
  ;[1 1 1] Time before enabling water interactions
  (Variables|/Script/FortniteGame.FortWaterInteractionComponent|SetbIsEnabled true (Variables|Water|GetWaterInteractionComponent)))

(event EventBeginPlay
  ;[1 1 1] Start rotation and rotations
  (Components|Movement|SetUpdatedComponent (Variables|Default|GetRotatingMovement) (Variables|Default|GetJunkMesh))
  (Utilities|Time|SetTimerbyEvent (delegate Custom|Explode) (ScalableFloat|GetValueatLevel (Variables|ProjectileInfo|GetExplodeTimer)))
  ;[1 1 1] Time before enabling water interactions
  (Utilities|Time|SetTimerbyEvent (delegate Custom|EnableWater) (Variables|ProjectileInfo|GetEnableWaterTimer)))

(event EventOnExploded
  (params (HitActors :type Array%20of%20Actor%20Object%20References) (HitResults :type Array%20of%20Hit%20Result%20Structures))
  ;[1 1 1] On Explode target and deal damage
  (|CallParentFunction HitActors HitResults)
  (Variables|ProjectileInfo|SetIsExploded true)
  (Utilities|FlowControl|DoOnce false
    (:Completed
      (Delivery|Kill)
      (Audio|Components|Audio|Stop (Variables|Default|GetLoopingAudioTell)))))

(event EventOnStop
  (params (Hit :type Hit%20Result%20Structure%20%28by%20ref%29))
  ;[1 1 1] On Hit Explode
  (CallFunction|TriggerExplode))

(event OnEnterWaterEvent(WaterInteractionComponent)
  (params (WaterBody :type Fort%20Water%20Body%20Actor%20Object%20Reference) (WaterInteractionComponent :type Fort%20Water%20Interaction%20Component%20Object%20Reference) (bIsFirstBody :type Boolean))
  ;[1 1 1] Explode shortly after entering water
  (CallFunction|TriggerExplode))

(event Custom|TriggerExplode
  ;[1 1 1] Explode shortly after entering water
  (if (not (Variables|ProjectileInfo|GetIsExploded))
    (Networking|SwitchHasAuthority
      (:Authority
        (Delivery|DoExplosionatLocation (Transformation|GetActorLocation self))
        (Variables|ProjectileInfo|SetRepEvent (+ (Variables|ProjectileInfo|GetRepEvent) 1))))))

(event Custom|CleanUp
  ;[1 1 1] Explode shortly after entering water
  (Networking|SwitchHasAuthority
    (:Remote
      (Rendering|SetVisibility (Variables|Default|GetJunkMesh) false false)
      (Components|Activation|Deactivate (Variables|Default|GetTrailFX))
      (Utilities|FlowControl|Delay 1.0)
      (Components|DestroyComponent (Variables|Default|GetTrailFX)))))

(fn OnRep_RepEvent
  (CallFunction|CleanUp))

(fn ConstructionScript)
