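# Unlit translucent light beam material for Junk Gun effects, enabled for Niagara mesh particles.
# Inverted vertex color red, scaled by separate color and alpha sharpness parameters, forms two edge masks.
# The color mask blends Color_End to Color_Start into MF_FX_EmissiveAdjustments for emissive.
# The alpha mask is multiplied by a powered absolute camera-vector and vertex-normal dot product, fading the mesh at grazing angles without a texture.
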Material /PrimalGameplay/Items/JunkGun/Effects/M_LightBeam
graph {
 properties {
  bUsedWithNiagaraMeshParticles=true
  blendMode=BLEND_Translucent
  pixelDepthOffsetMode=PDOM_Legacy
  shadingModel=MSM_Unlit
  usedShadingModels=Unlit
 }
 nodes [
  Constant4Vector() {constant=<rgba 1 1 1 1>}
  Constant() {r=1}
  n1 = Multiply(A=OneMinus(None=VertexColor().R), B=ScalarParameter() {defaultValue=0.4, parameterName=ColorEdgeMaskSharpness})
  MaterialFunctionCall(Emissive=LinearInterpolate(
    A=VectorParameter() {parameterName=Color_End, defaultValue=<rgba 0.3333 0.0107 0 1>}.RGB
    B=VectorParameter() {parameterName=Color_Start, defaultValue=<rgba 3 2.2609 0.5365 1>}.RGB
    Alpha=Saturate(None=@n1)
   )) {materialFunction=&MF_FX_EmissiveAdjustments}
  n2 = OneMinus(None=VertexColor().R)
  n3 = ScalarParameter() {defaultValue=0.3, parameterName=AlphaEdgeMaskSharpness}
  n4 = DotProduct(A=CameraVectorWS(), B=VertexNormalWS())
  Multiply(
   A=Saturate(None=Multiply(A=@n2, B=@n3))
   B=Power(
    Base=Abs(None=@n4)
    Exp=ScalarParameter() {defaultValue=2, desc=Exponent for frensel falloff., group="4) Fade", parameterName=Frenel Fade Exponent}
   )
  )
 ]
}
