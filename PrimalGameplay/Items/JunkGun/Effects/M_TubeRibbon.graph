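# Translucent unlit effect material for the JunkGun, flagged for Niagara ribbons and mesh particles, that renders an animated noisy tube.
# Three scaled, panning noise layers on swizzled UVs multiply into a mask that drives opacity and normal-based vertex displacement, added to a panning cylinder-gradient offset.
# Dynamic parameters supply UV offset and distortion intensity.
# Particle color feeds emissive; particle alpha, depth, near-camera and Fresnel fades soften opacity.
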Material /PrimalGameplay/Items/JunkGun/Effects/M_TubeRibbon
graph {
 properties {
  bUsedWithMeshDeformer=true
  bUsedWithNiagaraMeshParticles=true
  bUsedWithNiagaraRibbons=true
  bUsedWithSkeletalMesh=true
  blendMode=BLEND_Translucent
  pixelDepthOffsetMode=PDOM_Legacy
  shadingModel=MSM_Unlit
  usedShadingModels=Unlit
 }
 nodes [
  n1 = VertexNormalWS()
  n2 = Multiply(
   A=MaterialFunctionCall(XY=TextureCoordinate()) {materialFunction=&Swizzle}.YX
   B=AppendVector(
    A=ScalarParameter() {defaultValue=3, parameterName=ThirdTextureXScale}
    B=ScalarParameter() {defaultValue=0.3, parameterName=ThirdTextureYScale}
   )
  )
  n3 = ComponentMask(None=VectorParameter() {group=UVScale, parameterName=UVScale, defaultValue=<rgba 1 1 0 0>}.RGB) {g=true, r=true}
  n4 = DynamicParameter() {paramNames [UVOffset,WPOIntensity,Param3,Param4]}
  n5 = Panner(
   Coordinate=Add(A=Multiply(A=@n2, B=@n3), B=@n4.UVOffset)
   Speed=VectorParameter() {parameterName=ThirdTexturePanningSpeed, defaultValue=<rgba -0.01 0.8 0 1>}.RGB
  )
  n6 = TextureCoordinate()
  n7 = ScalarParameter() {defaultValue=4, parameterName=SecondTextureXScale}
  n8 = ScalarParameter() {defaultValue=0.4, parameterName=SecondTextureYScale}
  n9 = Multiply(
   A=@n3
   B=Multiply(
    A=MaterialFunctionCall(XY=@n6) {materialFunction=&Swizzle}.YX
    B=AppendVector(A=@n7, B=@n8)
   )
  )
  n10 = Multiply(A=ScalarParameter() {defaultValue=-1.5, parameterName=Speed}, B=Time())
  n11 = TextureSample(UVs=Panner(
    Coordinate=Add(A=@n9, B=@n4.UVOffset)
    Time=@n10
    Speed=VectorParameter() {parameterName=SecondTexturePanningSpeed, defaultValue=<rgba 0.05 1 0 1>}.RGB
   )) {texture=&T_NoisePack_01}
  n12 = Multiply(
   A=MaterialFunctionCall(
    In=TextureSample(UVs=@n5) {texture=&T_NoisePack_01}.G
    Contrast=ScalarParameter() {parameterName=ThirdTextureContrast}
   ) {materialFunction=&CheapContrast}.Result
   B=@n11.G
  )
  n13 = MaterialFunctionCall(XY=TextureCoordinate()) {materialFunction=&Swizzle}
  n14 = AppendVector(
   A=ScalarParameter() {defaultValue=8, parameterName=FirstTextureXScale}
   B=ScalarParameter() {defaultValue=0.9, parameterName=FirstTextureYScale}
  )
  n15 = Add(A=@n4.UVOffset, B=Multiply(A=@n3, B=Multiply(A=@n13.YX, B=@n14)))
  n16 = VectorParameter() {parameterName=FirstTexturePanningSpeed, defaultValue=<rgba 0.05 2 0 1>}
  n17 = Multiply(
   A=@n12
   B=TextureSample(UVs=Panner(Coordinate=@n15, Time=@n10, Speed=@n16.RGB)) {texture=&T_NoisePack_01}.G
  )
  n18 = Multiply(
   A=DynamicParameter() {paramNames [UVOffset,WPOIntensity,Param3,Param4]}.WPOIntensity
   B=ScalarParameter() {defaultValue=50, parameterName=Distort}
  )
  n19 = TextureCoordinate()
  n20 = Panner(
   Coordinate=Add(A=@n4.UVOffset, B=MaterialFunctionCall(XY=@n19) {materialFunction=&Swizzle}.YX)
   Time=@n10
   Speed=VectorParameter() {parameterName=WPOPanningSpeed, defaultValue=<rgba 0.05 2 0 1>}.RGB
  ) {speedY=0.25}
  Add(
   A=Multiply(A=Multiply(A=@n1, B=@n17), B=@n18)
   B=Multiply(A=TextureSample(UVs=@n20) {texture=/Game/Global/Textures/Default/Masks_and_Gradients/CylGradient}.RGB, B=@n18) {constB=35}
  )
  Add(A=@n12, B=@n11.G)
  Constant() {r=1}
  Constant() {r=1}
  n21 = ParticleColor()
  MaterialFunctionCall(Emissive=@n21.RGB) {materialFunction=&MF_FX_EmissiveAdjustments}
  n22 = Multiply(
   A=@n17
   B=MaterialFunctionCall(Distance=ScalarParameter() {defaultValue=250, parameterName=NearCameraFade}) {materialFunction=&Near_Camera_Fade}.Result
  )
  n23 = ScalarParameter() {defaultValue=64, parameterName=FadeDistance}
  n24 = Abs(None=DotProduct(A=CameraVectorWS(), B=VertexNormalWS()))
  n25 = ScalarParameter() {defaultValue=2, desc=Exponent for frensel falloff., group="4) Fade", parameterName=Frenel Fade Exponent}
  Multiply(A=@n21.A, B=Multiply(A=DepthFade(Opacity=@n22, FadeDistance=@n23) {fadeDistanceDefault=50}, B=Power(Base=@n24, Exp=@n25)))
  OneMinus(None=@n24)
  VertexNormalWS()
 ]
}
