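# Looping Niagara impact effect for the JunkGun, named for its full-ammo state, continuously spawning sprite debris, mesh debris, grains and smoke.
# All four emitters spawn on a torus aligned to User.Impact_Orientation, adding torus-normal and orientation velocity under gravity.
# User.CollectingMaterial sets VisibilityTag on both debris emitters' three renderers and indexes arrays for debris sprite size and brightness.
# Mesh debris and grains simulate on GPU; mesh debris also uses depth-buffer placement and distance culling.
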NiagaraSystem /PrimalGameplay/Items/JunkGun/Effects/NS_JunkGun_FullAmmo_Impact_loop
dependencies {
 usedDataInterfaces [&NiagaraDataInterfaceCurve, &NiagaraDataInterfaceGBuffer, &NiagaraDataInterfaceArrayFloat]
 usedDynamicInputs [&ReturnNormalizedExecIndex]
 usedModules [&SystemState]
}
emitters {
 DebrisSprite {
  emitterUpdateScript [
   &EmitterState()
   &SpawnRate(SpawnRate=8, Use Spawn Probability=true)
  ]
  particleSpawnScript [
   &InitializeParticle(
    Color Maximum=<rgba 0.9809 0.9598 0.877 1>
    Color Minimum=<rgba 0.4184 0.3263 0.3087 1>
    Color Mode=::NewEnumerator2
    Lifetime Max=1
    Lifetime Min=0.4
    Lifetime Mode=::NewEnumerator1
    Mesh Scale Mode=::NewEnumerator4
    Sprite Rotation Mode=::NewEnumerator2
    Sprite Size=&MakeVector2D(
     X=&SelectFloatFromArray(
      Direct Array Index=&MakeIntFromFloat(Float=&Subtract_Float(A=$User.CollectingMaterial, B=1))
      Float Selection Array=DataInterface() {propertyValues {FloatData [2.5,7,10]}}
     )
     Y=&SelectFloatFromArray(
      Direct Array Index=&MakeIntFromFloat(Float=&Subtract_Float(A=$User.CollectingMaterial, B=1))
      Float Selection Array=DataInterface() {propertyValues {FloatData [15,7,10]}}
     )
    )
    Sprite Size Mode=::NewEnumerator4
    Sprite UV Mode=::NewEnumerator3
    UsePositionOffset=false
    Write Material Random=true
   )
   /SetVariables(Particles.VisibilityTag=&MakeIntFromFloat(Float=$User.CollectingMaterial)) {isSetParameters=true}
   /SetVariables(Particles.RandomFloat=&RandomRangeFloat(Minimum=75, Maximum=95)) {isSetParameters=true}
   &TorusLocation(
    Align Torus With Axis=true
    Coordinate Space:ENiagaraCoordinateSpace=::World
    Large Radius=75
    MaskSpawnGroup=false
    Module.Override Randomness=false
    Torus Axis=$User.Impact_Orientation
    Torus Mode=::NewEnumerator2
   )
   &AddVelocity(
    Coordinate Space=::World
    Velocity=&Multiply_VectorByFloat(Vector=$Output.TorusLocation.TorusNormal, Float=&RandomRangeFloat(Minimum=-35, Maximum=-75))
   )
   &AddVelocity(
    Coordinate Space=::World
    Scale Added Velocity=$User.Impact_Orientation
    Velocity=&VectorFromFloat(Value=&RandomRangeFloat(Minimum=50, Maximum=100))
   ) {module=AddVelocity001}
   &DynamicMaterialParameters(
    Index 0 Param 1=&RandomRangeFloat(Minimum=0, Maximum=1)
    Module.Param0WriteEnabled=true
    Module.Param1WriteEnabled=true
    Module.Param2WriteEnabled=true
    Module.Param3WriteEnabled=true
   )
  ]
  particleUpdateScript [
   &GravityForce(Gravity=<0 0 -700>)
   &SolveForcesAndVelocity(Clamp Velocity=false, Limit Acceleration=false)
   &ParticleState()
   &ScaleSpriteSize(Scale Factor=&Vector2DFromFloat(Value=&FloatFromCurve(FloatCurve=DataInterface() {propertyValues {Curve {keys [
          {arriveTangent=2.8442, interpMode=RCIM_Cubic, leaveTangent=2.8442, tangentMode=RCTM_User, time=0, value=0}
          <at 0.5 1 RCIM_Cubic>
          {arriveTangent=-2.3002, interpMode=RCIM_Cubic, leaveTangent=-2.3002, tangentMode=RCTM_User, time=1, value=0}
         ]}}})))
   &SpriteRotationRate(Rotation Rate=&RandomRangeFloat(Minimum=-1000, Maximum=1000))
   &ScaleColor(
    Scale RGB=&VectorFromFloat(Value=&SelectFloatFromArray(
      Direct Array Index=&MakeIntFromFloat(Float=&Subtract_Float(A=$User.CollectingMaterial, B=1))
      Float Selection Array=DataInterface() {propertyValues {FloatData [0.75,0.5,1]}}
     ))
    ScaleA=true
    ScaleRGB=true
   )
  ]
  rendererClasses [&NiagaraSpriteRendererProperties]
  renderers [
   {rendererClass=&NiagaraSpriteRendererProperties}
   {rendererClass=&NiagaraSpriteRendererProperties, rendererIndex=1}
   {rendererClass=&NiagaraSpriteRendererProperties, rendererIndex=2}
  ]
 }
 Debris_3D {
  simTarget=GPUComputeSim
  emitterUpdateScript [
   &EmitterState(Enable Distance Culling=true, MaxDistance=1000, Module.bUseMaxDistance=true, Module.bUseMinDistance=false, Scalability Mode=::NewEnumerator1)
   &SpawnBurst_Instantaneous(Spawn Count=7, Use Spawn Probability=false) {enabled=false}
   &SpawnRate(SpawnRate=100, Use Spawn Probability=false)
  ]
  particleSpawnScript [
   /Niagara/Modules/Spawn/Initialization/InitializeParticle(
    Color=&RandomRangeLinearColor(Minimum=<rgba 0.4184 0.2702 0.2419 1>, Maximum=<rgba 0.9809 0.9598 0.877 1>)
    Lifetime=&UniformRangedFloat(Minimum=0.4, Maximum=0.65)
    Mesh Scale=&VectorFromFloat(Value=&RandomRangeFloat(Minimum=0.75, Maximum=1.2))
    Module.Write Color=true
    Module.Write Lifetime=true
    Module.Write Mass=true
    Module.Write Material Random=true
    Module.Write Position=true
    Module.Write Scale=true
    Module.Write SpriteRotation=false
    Module.Write SpriteSize=false
    Position=$User.ImpactLocation
   )
   /SetVariables(Particles.VisibilityTag=&MakeIntFromFloat(Float=$User.CollectingMaterial)) {isSetParameters=true}
   &CameraOffset(Camera Offset Amount=25) {enabled=false}
   /SetVariables(Particles.RandomFloat=&RandomRangeFloat(Minimum=75, Maximum=95)) {isSetParameters=true, module=SetVariables_#1}
   &TorusLocation(
    Align Torus With Axis=true
    Coordinate Space:ENiagaraCoordinateSpace=::World
    Large Radius=75
    MaskSpawnGroup=false
    Module.Override Randomness=false
    Torus Axis=$User.Impact_Orientation
    Torus Mode=::NewEnumerator2
   )
   /Niagara/Modules/Math/PlaceParticlesOnDepthBuffer_GPU(Max Z Depth=0, Module.Use Max Z Depth=false)
   &self:DistanceCullingCheck(MaximumDistance=1000, New Camera Query:NiagaraDataInterfaceCamera=$System.CameraPick) {module=DistanceCullingCheck}
   &KillParticlesInVolume(Sphere Radius=125)
   &AddVelocity(
    Coordinate Space=::World
    Velocity=&Multiply_VectorByFloat(Vector=$Output.TorusLocation.TorusNormal, Float=&RandomRangeFloat(Minimum=-35, Maximum=-75))
   )
   &AddVelocity(
    Coordinate Space=::World
    Scale Added Velocity=$User.Impact_Orientation
    Velocity=&VectorFromFloat(Value=&RandomRangeFloat(Minimum=50, Maximum=100))
   ) {module=AddVelocity001}
   &InitialMeshOrientation(
    Orientation Axis=&RandomVector(Vector Scale=1)
    Orientation Vector=&RandomVector(Vector Scale=1)
    Use Orientation Vector=true
    Use Rotation Vector=false
   )
   /Game/Effects/Niagara/Modules/Spawn/FN_Spawn_GBuffer_Color() {enabled=false}
  ]
  particleUpdateScript [
   &ParticleState()
   /SetVariables(Particles.VisibilityTag=&MakeCustomIntFromBool(Bool=$Transient.FirstFrame, True INT=1, False INT=0)) {enabled=false, isSetParameters=true}
   &GravityForce(Gravity=<0 0 -700>)
   &SolveForcesAndVelocity(Clamp Velocity=false, Limit Acceleration=false)
   &UpdateMeshOrientation(Rotation Rate=1.2, Rotation Vector=&RandomVector(Vector Scale=1))
   &ScaleMeshSize(Scale Factor=&VectorFromFloat(Value=&FloatFromCurve(FloatCurve=DataInterface() {propertyValues {Curve {keys [
          {arriveTangent=6.4077, interpMode=RCIM_Cubic, leaveTangent=6.4077, tangentMode=RCTM_User, time=0, value=0}
          <at 0.5 1 RCIM_Cubic>
          {arriveTangent=-3.3951, interpMode=RCIM_Cubic, leaveTangent=-3.3951, tangentMode=RCTM_User, time=1, value=0}
         ]}}})))
   &ScaleColor(Scale RGB=<10 10 10>, ScaleA=true, ScaleRGB=true) {enabled=false}
  ]
  rendererClasses [&NiagaraMeshRendererProperties]
  renderers [
   {rendererClass=&NiagaraMeshRendererProperties, rendererIndex=0}
   {rendererClass=&NiagaraMeshRendererProperties, rendererIndex=1}
   {rendererClass=&NiagaraMeshRendererProperties, rendererIndex=2}
  ]
 }
 Grains_Up {
  simTarget=GPUComputeSim
  emitterUpdateScript [
   &EmitterScalability(
    Module.bApplySpawnCountScale=false
    Module.bApplySpawnCountScaleByDistanceFraction=false
    Module.bUseMaxDistance=false
    Module.bUseMinDistance=false
   ) {enabled=false}
   &EmitterLifeCycle(NextLoopDuration=4.4208) {enabled=false}
   &EmitterState()
   &SpawnBurst_Instantaneous(Spawn Count=20, Use Spawn Probability=false) {enabled=false}
   &SpawnRate(SpawnRate=11, Use Spawn Probability=false)
  ]
  particleSpawnScript [
   &InitializeParticle(
    Color=&RandomRangeLinearColor(Minimum=<rgba 0.4184 0.2702 0.2419 1>, Maximum=<rgba 0.9809 0.9598 0.877 1>)
    Lifetime=&UniformRangedFloat(Minimum=0.5, Maximum=1)
    Mass=&RandomRangeFloat(Minimum=0.5, Maximum=1.5)
    Module.Write Color=true
    Module.Write Lifetime=true
    Module.Write Mass=true
    Module.Write Material Random=true
    Module.Write Position=true
    Module.Write Scale=false
    Module.Write SpriteRotation=true
    Module.Write SpriteSize=true
    Position=$User.ImpactLocation
    Sprite Rotation=&RandomRangeFloat(Minimum=0, Maximum=180)
    Sprite Size=&Vector2DFromFloat(Value=&RandomRangeFloat(Minimum=5, Maximum=8))
   )
   &TorusLocation(
    Align Torus With Axis=true
    Coordinate Space:ENiagaraCoordinateSpace=::World
    Large Radius:NiagaraFloat=&RandomRangeFloat(Minimum=100, Maximum=75)
    MaskSpawnGroup=false
    Module.Override Randomness=false
    Torus Axis=$User.Impact_Orientation
    Torus Mode=::NewEnumerator2
   )
   &AddVelocity(
    Coordinate Space=::World
    Velocity=&Multiply_VectorByFloat(Vector=$Output.TorusLocation.TorusNormal, Float=&RandomRangeFloat(Minimum=-25, Maximum=-100))
   ) {module=AddVelocity001}
   &AddVelocity(
    Coordinate Space=::World
    Scale Added Velocity=$User.Impact_Orientation
    Velocity=&VectorFromFloat(Value=&RandomRangeFloat(Minimum=75, Maximum=150))
   )
  ]
  particleUpdateScript [
   &ParticleState()
   &AccelerationForce(Acceleration=&Multiply_VectorByFloat(Vector=$User.Impact_Orientation, Float=500)) {enabled=false}
   &Drag(Use Linear Drag=true, Use Rotational Drag=false) {enabled=false}
   &GravityForce(Gravity=<0 0 -200>)
   &SolveForcesAndVelocity(Clamp Velocity=false, Limit Acceleration=false)
   &ScaleColor(
    Scale Alpha=&FloatFromCurve(FloatCurve=DataInterface() {propertyValues {Curve {keys [
         {arriveTangent=5.8721, interpMode=RCIM_Cubic, leaveTangent=5.8721, tangentMode=RCTM_User, time=0, value=0}
         <at 0.5 1 RCIM_Cubic>
         {arriveTangent=-4.764, interpMode=RCIM_Cubic, leaveTangent=-4.764, tangentMode=RCTM_User, time=1, value=0}
        ]}}})
    Scale RGB=<1.5 1.5 1.5>
    ScaleA=true
    ScaleRGB=true
   ) {enabled=false}
   &ScaleSpriteSize(Scale Factor=&Vector2DFromFloat(Value=&FloatFromCurve(FloatCurve=DataInterface() {propertyValues {Curve {keys [
          {arriveTangent=6.5802, interpMode=RCIM_Cubic, leaveTangent=6.5802, tangentMode=RCTM_User, time=0, value=0}
          {arriveTangent=0.002, interpMode=RCIM_Cubic, leaveTangent=0.002, time=0.2, value=1}
          {arriveTangent=-1.0897, interpMode=RCIM_Cubic, leaveTangent=-1.0897, time=0.7, value=1}
          {arriveTangent=-9.4926, interpMode=RCIM_Cubic, leaveTangent=-9.4926, tangentMode=RCTM_User, time=1.0011, value=0.002}
         ]}}})))
   &CameraOffset(Camera Offset Amount=10) {module=CameraOffset001}
  ]
  rendererClasses [&NiagaraSpriteRendererProperties]
  renderers [{rendererClass=&NiagaraSpriteRendererProperties}]
 }
 Smoke {
  emitterUpdateScript [
   &EmitterState()
   &SpawnBurst_Instantaneous(Spawn Count=7, Use Spawn Probability=false) {enabled=false}
   &SpawnRate(SpawnRate=5, Use Spawn Probability=false)
  ]
  particleSpawnScript [
   &InitializeParticle(
    Color=<rgba 0.283 0.2149 0.1626 0.5>
    Lifetime=&UniformRangedFloat(Minimum=0.8, Maximum=0.8)
    Module.Write Color=true
    Module.Write Lifetime=true
    Module.Write Mass=true
    Module.Write Material Random=true
    Module.Write Position=true
    Module.Write Scale=false
    Module.Write SpriteRotation=true
    Module.Write SpriteSize=true
    Position=$User.ImpactLocation
    Sprite Rotation=&NormalizedAngleToDegrees(NormalizedAngle=&UniformRangedFloat(Minimum=0, Maximum=1))
    Sprite Size=&Vector2DFromFloat(Value=&RandomRangeFloat(Minimum=80, Maximum=100))
   )
   &CameraOffset(Camera Offset Amount=&UniformRangedFloat(Minimum=-20, Maximum=20)) {enabled=false}
   /SetVariables(Particles.RandomFloat=&RandomRangeFloat(Minimum=0, Maximum=1)) {isSetParameters=true, module=SetVariables_#1}
   &TorusLocation(
    Align Torus With Axis=true
    Coordinate Space:ENiagaraCoordinateSpace=::World
    Large Radius=75
    MaskSpawnGroup=false
    Module.Override Randomness=false
    Torus Axis=$User.Impact_Orientation
    Torus Mode=::NewEnumerator2
   )
   &AddVelocity(
    Coordinate Space=::World
    Velocity=&Multiply_VectorByFloat(Vector=$Output.TorusLocation.TorusNormal, Float=&RandomRangeFloat(Minimum=-25, Maximum=-50))
   )
   &AddVelocity(
    Coordinate Space=::World
    Scale Added Velocity=$User.Impact_Orientation
    Velocity=&VectorFromFloat(Value=&RandomRangeFloat(Minimum=75, Maximum=150))
   ) {module=AddVelocity001}
   /SetVariables(Particles.RandomFrameStart:NiagaraInt32=&RandomRangeInt(Minimum=0, Maximum=63)) {isSetParameters=true}
  ]
  particleUpdateScript [
   &ParticleState()
   &AccelerationForce(Acceleration=&Multiply_VectorByFloat(Vector=$User.Impact_Orientation, Float=500)) {enabled=false}
   &PointForce(Force Strength=450, Linear Force Vector:Vector3f=$User.Impact_Orientation, Module.Use Falloff=false, Module.UseLinearForce=true) {enabled=false}
   &GravityForce(Gravity=<0 0 -200>)
   &SolveForcesAndVelocity(Clamp Velocity=false, Limit Acceleration=false)
   &ScaleSpriteSize(Scale Factor=&Vector2DFromFloat(Value=&FloatFromCurve(
      FloatCurve=DataInterface() {propertyValues {Curve {keys [
          {arriveTangent=0.7677, interpMode=RCIM_Cubic, leaveTangent=0.7677, tangentMode=RCTM_User, time=-0.003, value=0.3}
          {arriveTangent=0.0991, interpMode=RCIM_Cubic, leaveTangent=0.0991, tangentMode=RCTM_User, time=1, value=1}
         ]}}}
      Scale Curve=1.2
     )))
   &DynamicMaterialParameters(
    Index 0 Param 1=&FloatFromCurve(FloatCurve=DataInterface() {propertyValues {Curve {keys [<at 0 1>,<at 1 0>]}}}, Scale Curve=0.5)
    Index 2 Param 1=&FloatFromCurve(FloatCurve=DataInterface() {propertyValues {Curve {keys [
         {arriveTangent=-10.5859, interpMode=RCIM_Cubic, leaveTangent=-10.5859, tangentMode=RCTM_User, time=0, value=2}
         <at 0.2 1 RCIM_Cubic>
         <at 1 2 RCIM_Cubic>
        ]}}})
    Index 2 Param 2=&FloatFromCurve(FloatCurve=DataInterface() {propertyValues {Curve {keys [<at 0 1>,<at 1 0>]}}}, Scale Curve=0.5)
    Index 2 Param 3=1
    Index 2 Param 4=&FloatFromCurve(FloatCurve=DataInterface() {propertyValues {Curve {keys [<at 0 1.2 RCIM_Cubic>,<at 0.2 0.8 RCIM_Cubic>,<at 1 1.2 RCIM_Cubic>]}}})
    Module.Param0WriteEnabled=true
    Module.Param0WriteEnabled2=true
    Module.Param1WriteEnabled=false
    Module.Param1WriteEnabled2=true
    Module.Param2WriteEnabled=true
    Module.Param2WriteEnabled2=true
    Module.Param3WriteEnabled=true
    Module.Param3WriteEnabled2=true
    Write Parameter Index 2=true
   )
   &ScaleColor(
    Scale Alpha=&FloatFromCurve(FloatCurve=DataInterface() {propertyValues {Curve {keys [
         {arriveTangent=1.3525, interpMode=RCIM_Cubic, leaveTangent=1.3525, tangentMode=RCTM_User, time=0, value=0}
         {arriveTangent=1.3571, interpMode=RCIM_Cubic, leaveTangent=1.3571, time=0.4, value=0.925}
         {arriveTangent=-0.3287, interpMode=RCIM_Cubic, leaveTangent=-0.3287, tangentMode=RCTM_User, time=0.7, value=0.95}
         {arriveTangent=-3.3944, interpMode=RCIM_Cubic, leaveTangent=-3.3944, tangentMode=RCTM_User, time=1, value=0}
        ]}}})
    ScaleA=true
    ScaleRGB=true
   )
   &SpriteRotationRate(Rotation Rate=&Multiply_Float(
     A=$Particles.InitSpriteRotRate
     B=&FloatFromCurve(FloatCurve=DataInterface() {propertyValues {Curve {keys [<at 0 0.09 RCIM_Cubic>,<at 1.0011 1 RCIM_Cubic>]}}})
    )) {enabled=false}
   &SubUVAnimation(
    End Frame:NiagaraInt32=&Add_Integer(A=$Particles.RandomFrameStart, B=25)
    Start Frame:NiagaraInt32=$Particles.RandomFrameStart
   ) {enabled=false}
   &CameraOffset(Camera Offset Amount=30) {module=CameraOffset001}
  ]
  rendererClasses [&NiagaraSpriteRendererProperties]
  renderers [
   {rendererClass=&NiagaraSpriteRendererProperties}
   {rendererClass=&NiagaraSpriteRendererProperties, rendererIndex=1}
  ]
 }
}
summary {
 systemName=NS_JunkGun_FullAmmo_Impact_loop
 emitters [
  {emitterName=Smoke, rendererClasses [&NiagaraSpriteRendererProperties]}
  {emitterName=Grains_Up, simTarget=GPUComputeSim, rendererClasses [&NiagaraSpriteRendererProperties]}
  {emitterName=Debris_3D, simTarget=GPUComputeSim, rendererClasses [&NiagaraMeshRendererProperties]}
  {emitterName=DebrisSprite, rendererClasses [&NiagaraSpriteRendererProperties]}
 ]
 userVariables [
  User.bIsHitting {type=NiagaraBool}
  User.CollectingMaterial {type=NiagaraFloat, defaultValue=2, description=""}
  User.Impact_Orientation {type=Vector3f, description="", defaultValue {struct=Vector3f, value=<1 0 0>}}
  User.ImpactLocation {type=NiagaraPosition, description="", defaultValue {}}
 ]
}
