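# Niagara system for the JunkGun rocket launcher's ammo refill, pulling short-lived debris chunks from an impact point to the weapon.
# AmmoRefillChildStone bursts four spinning, size-varied meshes and sets a visibility tag from User.CollectingMaterial, apparently selecting among six mesh renderers.
# Glow adds cyan sprites and lights on the same curve-driven lerp from ImpactLocation to WeaponLocation, which tracks the weapon each update.
# The Decal emitter and SkeletalMeshLocation modules are disabled.
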NiagaraSystem /PrimalGameplay/Items/JunkGun/Effects/NS_RocketLauncher_JunkGun_Ammo
dependencies {
 usedDataInterfaces [&NiagaraDataInterfaceArrayFloat, &NiagaraDataInterfaceArrayFloat3, &NiagaraDataInterfaceCurve]
 usedDynamicInputs [
  &SelectFloatFromArray
  &VectorLength
  &Subtract_Vector
  &Multiply_Float
  /Niagara/DynamicInputs/Transforms/TransformPosition
 ]
 usedModules [
  &SystemState
  &self:SystemSpawnScript.SystemScriptSource.SystemScriptGraph.NiagaraNodeAssignment_0.SetVariables_#1
  &self:SystemSpawnScript.SystemScriptSource.SystemScriptGraph.NiagaraNodeAssignment_2.SetVariables_#2
  &self:SystemSpawnScript.SystemScriptSource.SystemScriptGraph.NiagaraNodeAssignment_1.SetVariables_#3
  &self:SystemSpawnScript.SystemScriptSource.SystemScriptGraph.NiagaraNodeAssignment_3.SetVariables_#4
  &self:SystemSpawnScript.SystemScriptSource.SystemScriptGraph.NiagaraNodeAssignment_4.SetVariables_#5
 ]
}
emitters {
 AmmoRefillChildStone {
  emitterUpdateScript [
   &EmitterState()
   &SpawnBurst_Instantaneous(Spawn Count=4, Use Spawn Probability=false)
  ]
  particleSpawnScript [
   /SetVariables(Particles.ImpactPosition:Vector3f=$User.ImpactLocation) {isSetParameters=true, module=SetVariables_#6}
   /SetVariables(Particles.VisibilityTag=&MakeIntFromFloat(Float=$User.CollectingMaterial)) {isSetParameters=true}
   &InitializeParticle(
    Color Maximum=<rgba 0.9809 0.9598 0.877 1>
    Color Minimum=<rgba 0.4184 0.3056 0.284 1>
    Color Mode=::NewEnumerator2
    Lifetime=0.65
    Mesh Scale=&SelectVectorFromArray(
     Direct Array Index=&ReturnExecIndex()
     Vector Selection Array=DataInterface() {propertyValues {FloatData [<0.35 0.35 1>, <1.5 1 1>, <0.8 0.6 0.6>, <1.1 1.1 0.85>]}}
    )
    Mesh Scale Mode=::NewEnumerator4
    Sprite Size Mode=::NewEnumerator3
    Uniform Sprite Size=25
    UsePositionOffset=false
    Write Material Random=true
   )
   &InitialMeshOrientation(
    Orientation Axis=&RandomVector(Vector Scale=1)
    Orientation Vector=&RandomVector(Vector Scale=1)
    Rotation=&RandomVector(Vector Scale=1)
    Rotation Coordinate Space=::NewEnumerator1
    Use Orientation Vector=true
    Use Rotation Vector=true
   )
   /SetVariables(Particles.Renderer2_Color:LinearColor=<rgba 0.6163 2.6946 5 1>) {isSetParameters=true}
  ]
  particleUpdateScript [
   &ParticleState()
   /SetVariables(Particles.WeaponPosition:Vector3f=$User.WeaponLocation) {isSetParameters=true, module=SetVariables_#7}
   /SetVariables(Particles.Position=&Lerp_Vector(
     A=$Particles.ImpactPosition
     B=$User.WeaponLocation
     Alpha=&FloatFromCurve(FloatCurve=DataInterface() {propertyValues {Curve {keys [<at 0 0 RCIM_Cubic>,<at 1 1 RCIM_Cubic>]}}})
    )) {isSetParameters=true, module=SetVariables_#8}
   &ScaleMeshSize(Scale Factor=&VectorFromFloat(Value=&FloatFromCurve(FloatCurve=DataInterface() {propertyValues {Curve {keys [
          {arriveTangent=9.4043, interpMode=RCIM_Cubic, leaveTangent=9.4043, tangentMode=RCTM_User, time=0, value=0}
          {arriveTangent=1.6429, interpMode=RCIM_Cubic, leaveTangent=1.6429, tangentMode=RCTM_User, time=0.15, value=1}
          {arriveTangent=-1.298, interpMode=RCIM_Cubic, leaveTangent=-1.298, tangentMode=RCTM_User, time=0.6, value=0.9}
          {arriveTangent=-1.057, interpMode=RCIM_Cubic, leaveTangent=-1.057, tangentMode=RCTM_User, time=1, value=0}
         ]}}})))
   &UpdateMeshOrientation(Rotation Vector=&RandomVector(Vector Scale=1))
   &DynamicMaterialParameters(
    Index 0 Param 2=2
    Index 0 Param 3=0.5
    Module.Param0WriteEnabled=true
    Module.Param1WriteEnabled=true
    Module.Param2WriteEnabled=true
    Module.Param3WriteEnabled=true
   )
  ]
  rendererClasses [&NiagaraMeshRendererProperties]
  renderers [
   {rendererClass=&NiagaraMeshRendererProperties, rendererIndex=0}
   {rendererClass=&NiagaraMeshRendererProperties, rendererIndex=1}
   {rendererClass=&NiagaraMeshRendererProperties, rendererIndex=2}
   {rendererClass=&NiagaraMeshRendererProperties, rendererIndex=3}
   {rendererClass=&NiagaraMeshRendererProperties, rendererIndex=4}
   {rendererClass=&NiagaraMeshRendererProperties, rendererIndex=5}
  ]
 }
 Decal {
  bEnabled=false
  emitterUpdateScript [
   &EmitterState()
   &SpawnBurst_Instantaneous(Spawn Count=4, Use Spawn Probability=false)
  ]
  particleSpawnScript [
   /SetVariables(Particles.ImpactPosition:Vector3f=$User.ImpactLocation) {isSetParameters=true, module=SetVariables_#6}
   /SetVariables(Particles.WeaponPosition:Vector3f=$User.WeaponLocation) {isSetParameters=true, module=SetVariables_#7}
   &InitializeParticle(
    Color=<rgba 0.6163 2.6946 5 1>
    Lifetime=0.65
    Sprite Size Mode=::NewEnumerator3
    Uniform Sprite Size=75
    UsePositionOffset=false
    Write Material Random=true
   )
  ]
  particleUpdateScript [
   &ParticleState()
   &SkeletalMeshLocation(
    Apply To Particle Position=false
    Inherit Velocity=false
    MaskSpawnGroup=false
    Mesh Sampling Type=::NewEnumerator5
    Sampled Position Offset Coordinate Space:ENiagaraMeshTransforms=::NewEnumerator3
    Skeletal Mesh=$User.Skeletal Mesh
    Socket Sampling Mode=::NewEnumerator5
   ) {enabled=false}
   /SetVariables(Particles.Position=&Lerp_Vector(
     A=$Particles.ImpactPosition
     B=$User.WeaponLocation
     Alpha=&FloatFromCurve(FloatCurve=DataInterface() {propertyValues {Curve {keys [<at 0 0 RCIM_Cubic>,<at 1 1 RCIM_Cubic>]}}})
    )) {isSetParameters=true, module=SetVariables_#8}
   &ScaleSpriteSize(Scale Factor=&Vector2DFromFloat(Value=&FloatFromCurve(FloatCurve=DataInterface() {propertyValues {Curve {keys [
          {arriveTangent=9.4043, interpMode=RCIM_Cubic, leaveTangent=9.4043, tangentMode=RCTM_User, time=0, value=0}
          {arriveTangent=1.6429, interpMode=RCIM_Cubic, leaveTangent=1.6429, tangentMode=RCTM_User, time=0.15, value=1}
          {arriveTangent=-0.4385, interpMode=RCIM_Cubic, leaveTangent=-0.4385, tangentMode=RCTM_User, time=0.5236, value=1}
          {arriveTangent=-2.2986, interpMode=RCIM_Cubic, leaveTangent=-2.2986, tangentMode=RCTM_User, time=1, value=0.3}
         ]}}})))
  ]
  rendererClasses [&NiagaraSpriteRendererProperties]
  renderers [{rendererClass=&NiagaraSpriteRendererProperties}]
 }
 Glow {
  emitterUpdateScript [
   &EmitterState()
   &SpawnBurst_Instantaneous(Spawn Count=4, Use Spawn Probability=false)
  ]
  particleSpawnScript [
   /SetVariables(Particles.ImpactPosition:Vector3f=$User.ImpactLocation) {isSetParameters=true, module=SetVariables_#6}
   &InitializeParticle(
    Color=<rgba 0.6163 2.6946 5 0.1>
    Lifetime=0.65
    Sprite Size Mode=::NewEnumerator3
    Uniform Sprite Size=75
    UsePositionOffset=false
    Write Material Random=true
   )
  ]
  particleUpdateScript [
   &ParticleState()
   /SetVariables(Particles.WeaponPosition:Vector3f=$User.WeaponLocation) {isSetParameters=true, module=SetVariables_#7}
   &SkeletalMeshLocation(
    Apply To Particle Position=false
    Inherit Velocity=false
    MaskSpawnGroup=false
    Mesh Sampling Type=::NewEnumerator5
    Sampled Position Offset Coordinate Space:ENiagaraMeshTransforms=::NewEnumerator3
    Skeletal Mesh=$User.Skeletal Mesh
    Socket Sampling Mode=::NewEnumerator5
   ) {enabled=false}
   /SetVariables(Particles.Position=&Lerp_Vector(
     A=$Particles.ImpactPosition
     B=$User.WeaponLocation
     Alpha=&FloatFromCurve(FloatCurve=DataInterface() {propertyValues {Curve {keys [<at 0 0 RCIM_Cubic>,<at 1 1 RCIM_Cubic>]}}})
    )) {isSetParameters=true, module=SetVariables_#8}
   &ScaleSpriteSize(Scale Factor=&Vector2DFromFloat(Value=&FloatFromCurve(FloatCurve=DataInterface() {propertyValues {Curve {keys [
          {arriveTangent=9.4043, interpMode=RCIM_Cubic, leaveTangent=9.4043, tangentMode=RCTM_User, time=0, value=0}
          {arriveTangent=1.6429, interpMode=RCIM_Cubic, leaveTangent=1.6429, tangentMode=RCTM_User, time=0.15, value=1}
          {arriveTangent=-0.4385, interpMode=RCIM_Cubic, leaveTangent=-0.4385, tangentMode=RCTM_User, time=0.5236, value=1}
          {arriveTangent=-2.2986, interpMode=RCIM_Cubic, leaveTangent=-2.2986, tangentMode=RCTM_User, time=1, value=0.3}
         ]}}})))
   &FN_Light_Attributes(
    Color=<rgba 0.3082 1.3473 2.5 1>
    Create Color Attribute=true
    Create Exponent Attribute=true
    Create Radius Attribute=true
    Exponent=5
    Radius=200
   )
  ]
  rendererClasses [&NiagaraSpriteRendererProperties, &NiagaraLightRendererProperties]
  renderers [
   {rendererClass=&NiagaraSpriteRendererProperties}
   {rendererClass=&NiagaraLightRendererProperties, rendererIndex=1}
  ]
 }
}
summary {
 systemName=NS_RocketLauncher_JunkGun_Ammo
 emitters [
  {emitterName=AmmoRefillChildStone, rendererClasses [&NiagaraMeshRendererProperties]}
  {emitterName=Glow, rendererClasses [&NiagaraSpriteRendererProperties, &NiagaraLightRendererProperties]}
  {bEnabled=false, emitterName=Decal, rendererClasses [&NiagaraSpriteRendererProperties]}
 ]
 userVariables [
  User.AlphaMult {type=NiagaraFloat, defaultValue=1, description=""}
  User.bIsCollecting {type=NiagaraBool}
  User.bIsHitting {type=NiagaraBool, defaultValue=true}
  User.CollectingMaterial {type=NiagaraFloat, defaultValue=3, description=""}
  User.CorrectDirection {type=NiagaraFloat, defaultValue=0, description=""}
  User.Impact_Orientation {type=Vector3f, description="", defaultValue {struct=Vector3f, value=<1 0 0>}}
  User.ImpactLocation {type=Vector3f, description="", defaultValue {struct=Vector3f, value=<150 0 0>}}
  User.RibbonIDOffset {type=NiagaraInt32}
  User.Skeletal Mesh {
   type=NiagaraDataInterfaceSkeletalMesh
   defaultValue {
    struct=NiagaraExt_VariableValue_DataInterface
    value {dataInterface=&self:NiagaraDataInterfaceSkeletalMesh_2}
   }
  }
  User.SplineUpdatedEndPos {type=Vector3f, description="", defaultValue {struct=Vector3f, value=<0 0 0>}}
  User.WeaponLocation {type=Vector3f, description="", defaultValue {struct=Vector3f, value=<0 0 0>}}
 ]
}
