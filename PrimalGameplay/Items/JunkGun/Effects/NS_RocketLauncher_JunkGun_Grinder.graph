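# Niagara effect for the JunkGun rocket launcher's grinder, with a fiery spinning turbine and a hopper that throws debris.
# Three fire emitters step around a circle in 16 segments, culling part of each turn; Long and Middle follow System.TurbineSpinSpeed, Thin uses its own speed.
# Turbine_Glow orbits a sprite and light likewise.
# When User.bIsHitting is true, socket-sampled hopper emitters add sine-pulsed sparks and mesh debris, plus flickering glow.
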NiagaraSystem /PrimalGameplay/Items/JunkGun/Effects/NS_RocketLauncher_JunkGun_Grinder
dependencies {
 usedDataInterfaces [&NiagaraDataInterfaceCurve]
 usedDynamicInputs []
 usedModules [
  &self:SystemSpawnScript.SystemScriptSource.SystemScriptGraph.NiagaraNodeAssignment_1.SetVariables_#1
  &self:SystemSpawnScript.SystemScriptSource.SystemScriptGraph.NiagaraNodeAssignment_2.SetVariables_#2
  &SystemState
  &self:SystemSpawnScript.SystemScriptSource.SystemScriptGraph.NiagaraNodeAssignment_0.SetVariables_#3
 ]
}
emitters {
 Empty003 {
  bEnabled=false
  emitterUpdateScript [
   &EmitterState()
   &SpawnBurst_Instantaneous(
    Spawn Count=1
    Spawn Probability=&MakeCustomFloatFromBool(
     Bool=&SetBoolByIntComparison(A=$Emitter.LoopCount, B=0, Comparison Type=::NewEnumerator0)
     True Float=0
     False Float=1
    )
    Use Spawn Probability=true
   )
  ]
  particleSpawnScript [
   &InitializeParticle(
    Color=<rgba 5 1 0 1>
    Mesh Scale=<0.05 0.05 0.25>
    Mesh Scale Mode=::NewEnumerator4
    Sprite Size Mode=::NewEnumerator3
    UsePositionOffset=false
    Write Material Random=true
   )
   &SkeletalMeshLocation(
    Inherit Velocity=false
    MaskSpawnGroup=false
    Mesh Sampling Type=::NewEnumerator5
    Skeletal Mesh=$User.SkeletalMesh
    Socket Sampling Mode=::NewEnumerator5
   )
  ]
  particleUpdateScript [
   /SetVariables(Particles.Position=&Multiply_VectorByFloat(
     Vector=&MakeVector(
      X=10
      Y=&Sine(Normalized Angle=&Multiply_Float(A=$Emitter.Age, B=$System.TurbineSpinSpeed), Period=-1, Scale=1, Bias=0)
      Z=&Cosine(Normalized Angle=&Multiply_Float(A=$Emitter.Age, B=$System.TurbineSpinSpeed), Period=-1, Scale=1, Bias=0)
     )
     Float=1
    )) {isSetParameters=true, module=SetVariables_#4}
   &ParticleState(
    Kill Particles When Lifetime Has Elapsed=false
    Let Infinitely Lived Particles Die When Emitter Deactivates=true
    Loop Particles Lifetime=true
   )
   &ScaleColor(
    Scale Alpha=&FloatFromCurve(FloatCurve=DataInterface() {propertyValues {Curve {keys [{arriveTangent=-2.3614, interpMode=RCIM_Cubic, leaveTangent=-2.3614, tangentMode=RCTM_User, time=0, value=1}, <at 1 0 RCIM_Cubic>]}}})
    ScaleA=true
    ScaleRGB=true
   )
  ]
  rendererClasses [&NiagaraLightRendererProperties]
  renderers [{rendererClass=&NiagaraLightRendererProperties, rendererIndex=0}]
 }
 Hopper_Debris {
  emitterUpdateScript [
   &EmitterState()
   &SpawnRate(
    Spawn Probability=$System.SpawnProbability
    SpawnRate=&MakeCustomFloatFromBool(
     Bool=$User.bIsHitting
     True Float=&Sine(Normalized Angle=$Emitter.Age, Period=0.3, Scale=5, Bias=5)
     False Float=0
    )
    Use Spawn Probability=true
   )
  ]
  particleSpawnScript [
   /SetVariables(Particles.MeshIndex=&RandomRangeInt(Minimum=0, Maximum=2)) {isSetParameters=true}
   &InitializeParticle(
    Color Mode=::NewEnumerator0
    Lifetime=&RandomRangeFloat(Minimum=0.3, Maximum=0.6)
    Mesh Scale Mode=::NewEnumerator1
    Mesh Uniform Scale Max=0.8
    Mesh Uniform Scale Min=0.5
    UsePositionOffset=false
    Write Material Random=true
   )
   &SkeletalMeshLocation(
    Bone / Socket Index=1
    Inherit Velocity=false
    MaskSpawnGroup=false
    Mesh Sampling Type=::NewEnumerator5
    Skeletal Mesh=$User.SkeletalMesh
    Socket Sampling Mode=::NewEnumerator5
   )
   &SphereLocation(
    MaskSpawnGroup=false
    Offset=<0 0 -2>
    Override Seed=false
    Sphere Radius=3
    Surface Only=false
    UseNonUniformScale=false
    UseOffset=true
    UseSphereOrientation=false
   )
   &AddVelocityInCone(
    Cone Angle=60
    Cone Axis=<0 0 1>
    Module.Use Velocity Falloff On Cone Axis=false
    Velocity Strength=&Multiply_Float(
     A=&RandomRangeFloat(Minimum=75, Maximum=200)
     B=&Sine(Normalized Angle=$Emitter.Age, Period=0.5, Scale=1, Bias=1.5)
    )
   )
   &InitialMeshOrientation(
    Orientation Axis=&RandomRangeVector(Minimum=<-1 -1 -1>, Maximum=<1 1 1>)
    Use Orientation Vector=true
    Use Rotation Vector=false
   )
  ]
  particleUpdateScript [
   &ParticleState()
   &Drag(
    Drag=&Lerp_Float(
     A=1
     B=3
     Alpha=&NormalizeFloat(Value=&VectorLength(Vector=$Particles.Velocity), Min=0, Max=100)
    )
    Use Linear Drag=true
    Use Rotational Drag=false
   )
   &GravityForce(Gravity=<0 0 -500>)
   &SolveForcesAndVelocity(Clamp Velocity=false, Limit Acceleration=false)
   &ScaleMeshSize(Scale Factor=&VectorFromFloat(Value=&FloatFromCurve(FloatCurve=DataInterface() {propertyValues {Curve {keys [
          {arriveTangent=2.8442, interpMode=RCIM_Cubic, leaveTangent=2.8442, tangentMode=RCTM_User, time=0, value=0}
          <at 0.5 1 RCIM_Cubic>
          {arriveTangent=-2.3002, interpMode=RCIM_Cubic, leaveTangent=-2.3002, tangentMode=RCTM_User, time=1, value=0}
         ]}}})))
   &UpdateMeshOrientation(Rotation Rate=2, Rotation Vector=<1 1 0>)
  ]
  rendererClasses [&NiagaraMeshRendererProperties]
  renderers [{rendererClass=&NiagaraMeshRendererProperties, rendererIndex=0}]
 }
 Hopper_Glow {
  emitterUpdateScript [
   &EmitterState()
   &SpawnRate(
    Spawn Probability=$System.SpawnProbability
    SpawnRate=&MakeCustomFloatFromBool(Bool=$User.bIsHitting, True Float=3, False Float=0)
    Use Spawn Probability=true
   )
  ]
  particleSpawnScript [
   &InitializeParticle(
    Color=<rgba 5 0.7083 0 0.05>
    Lifetime=0.3
    Sprite Size Mode=::NewEnumerator1
    Uniform Sprite Size Max=40
    Uniform Sprite Size Min=30
    UsePositionOffset=false
    Write Material Random=true
   )
   &SkeletalMeshLocation(
    Bone / Socket Index=1
    Inherit Velocity=false
    MaskSpawnGroup=false
    Mesh Sampling Type=::NewEnumerator5
    Sampled Position Offset=<0 0 3>
    Skeletal Mesh=$User.SkeletalMesh
    Socket Sampling Mode=::NewEnumerator5
   )
  ]
  particleUpdateScript [
   &ParticleState()
   &ScaleColor(
    Scale Alpha=&FloatFromCurve(FloatCurve=DataInterface() {propertyValues {Curve {keys [{arriveTangent=-2.3614, time=0, value=0}, <at 0.5 1>, <at 1 0>]}}}, Scale Curve=0.7)
    Scale RGB=&VectorFromFloat(Value=&Sine(
      Normalized Angle=$Emitter.Age
      Period=&RandomRangeFloat(Minimum=0.15, Maximum=0.5)
      Scale=0.45
      Bias=0.65
     ))
    ScaleA=true
    ScaleRGB=true
   )
   &ScaleSpriteSize(Scale Factor=&Vector2DFromFloat(Value=&FloatFromCurve(FloatCurve=DataInterface() {propertyValues {Curve {keys [<at 0 0.5>,<at 0.5 1>,<at 1 0.5>]}}})))
   &CameraOffset(Camera Offset Amount=10)
   &FN_Light_Attributes(
    Color=$Particles.Color
    Create Color Attribute=true
    Create Exponent Attribute=true
    Create Position Attribute=true
    Create Radius Attribute=true
    Exponent=12
    Position=&Add_Vector(A=Linked() {linkedVariable {name=Particles.Position, type=NiagaraPosition}}, B=<0 0 10>)
    Radius=35
    Scale Color=10
   )
  ]
  rendererClasses [&NiagaraSpriteRendererProperties, &NiagaraLightRendererProperties]
  renderers [
   {rendererClass=&NiagaraSpriteRendererProperties}
   {rendererClass=&NiagaraLightRendererProperties, rendererIndex=1}
  ]
 }
 Hopper_Sparks {
  emitterUpdateScript [
   &EmitterState()
   &SpawnRate(
    Spawn Probability=$System.SpawnProbability
    SpawnRate=&MakeCustomFloatFromBool(
     Bool=$User.bIsHitting
     True Float=&Sine(Normalized Angle=$Emitter.Age, Period=0.3, Scale=50, Bias=50)
     False Float=0
    )
    Use Spawn Probability=true
   )
  ]
  particleSpawnScript [
   &InitializeParticle(
    Color Maximum=<rgba 75 10 0 1>
    Color Minimum=<rgba 25 3.3333 0 1>
    Color Mode=::NewEnumerator2
    Lifetime=&RandomRangeFloat(Minimum=0.05, Maximum=0.25)
    Sprite Size Max=<2 4>
    Sprite Size Min=<1 1>
    Sprite Size Mode=::NewEnumerator2
    UsePositionOffset=false
    Write Material Random=true
   )
   &SkeletalMeshLocation(
    Bone / Socket Index=1
    Inherit Velocity=false
    MaskSpawnGroup=false
    Mesh Sampling Type=::NewEnumerator5
    Skeletal Mesh=$User.SkeletalMesh
    Socket Sampling Mode=::NewEnumerator5
   )
   &SphereLocation(
    MaskSpawnGroup=false
    Offset=<0 0 -2>
    Override Seed=false
    Sphere Radius=3
    Surface Only=false
    UseNonUniformScale=false
    UseOffset=true
    UseSphereOrientation=false
   )
   &AddVelocityInCone(
    Cone Angle=60
    Cone Axis=<0 0 1>
    Module.Use Velocity Falloff On Cone Axis=false
    Velocity Strength=&Multiply_Float(
     A=&RandomRangeFloat(Minimum=100, Maximum=300)
     B=&Sine(Normalized Angle=$Emitter.Age, Period=0.5, Scale=1, Bias=1.5)
    )
   )
  ]
  particleUpdateScript [
   &CurlNoiseForce(
    Noise Frequency=75
    Noise Strength=&FloatFromCurve(FloatCurve=DataInterface() {propertyValues {Curve {keys [<at 0 1>,<at 0.3 0>]}}}, Scale Curve=8000)
    Pan Noise=true
    Pan Noise Field=<0 0 5>
   )
   &Drag(
    Drag=&Lerp_Float(
     A=2
     B=6
     Alpha=&NormalizeFloat(Value=&VectorLength(Vector=$Particles.Velocity), Min=0, Max=100)
    )
    Use Linear Drag=true
    Use Rotational Drag=false
   )
   &SolveForcesAndVelocity(Clamp Velocity=false, Limit Acceleration=false)
   &ParticleState()
   &ScaleSpriteSizeBySpeed(Max Scale Factor=<1 6>)
   &ScaleColor(
    Scale Alpha=&FloatFromCurve(FloatCurve=DataInterface() {propertyValues {Curve {keys [{arriveTangent=-2.3614, interpMode=RCIM_Cubic, leaveTangent=-2.3614, tangentMode=RCTM_User, time=0, value=1}, <at 1 0 RCIM_Cubic>]}}})
    Scale RGB=&MakeVectorFromLinearColorRGB(Color=<rgba 1 0.4573 0 1>)
    ScaleA=true
    ScaleRGB=true
   )
   &ScaleSpriteSize(Scale Factor=&MakeVector2D(X=1, Y=&FloatFromCurve(FloatCurve=DataInterface() {propertyValues {Curve {keys [<at 0 1>,<at 1 0>]}}})))
  ]
  rendererClasses [&NiagaraSpriteRendererProperties]
  renderers [{rendererClass=&NiagaraSpriteRendererProperties}]
 }
 LongFire {
  emitterUpdateScript [
   &EmitterState()
   &SpawnRate(SpawnRate=50, Use Spawn Probability=false)
  ]
  particleSpawnScript [
   /SetVariables(Transient.SpinSegments=16) {isSetParameters=true, module=SetVariables_#5}
   /SetVariables(Transient.SpinLocation:NiagaraFloat=&Divide_Float(
     A=&Ceil_Float(Float=&Multiply_Float(
       A=&Frac_Float(Float=&Multiply_Float(A=$Emitter.Age, B=$System.TurbineSpinSpeed))
       B=$Transient.SpinSegments
      ))
     B=$Transient.SpinSegments
    )) {isSetParameters=true, module=SetVariables_#6}
   &InitializeParticle(
    Color=<rgba 1 0.1 0 1>
    Lifetime=0.3
    Mesh Scale=<0.05 0.05 0.25>
    Mesh Scale Mode=::NewEnumerator4
    Sprite Size Mode=::NewEnumerator3
    UsePositionOffset=false
    Write Material Random=true
   )
   &SkeletalMeshLocation(
    Inherit Velocity=false
    MaskSpawnGroup=false
    Mesh Sampling Type=::NewEnumerator5
    Skeletal Mesh=$User.SkeletalMesh
    Socket Sampling Mode=::NewEnumerator5
   )
   /SetVariables() {isSetParameters=true, module=SetVariables_#7}
   /SetVariables(Particles.DEBUG_MeshSize:Vector3f=&VectorFromFloat(Value=0.25)) {isSetParameters=true, module=SetVariables_#8}
   /SetVariables(Particles.Position=&Multiply_VectorByFloat(
     Vector=&MakeVector(
      X=0.2
      Y=&Sine(Normalized Angle=$Transient.SpinLocation, Period=-1, Scale=1, Bias=0)
      Z=&Cosine(Normalized Angle=$Transient.SpinLocation, Period=-1, Scale=1, Bias=0)
     )
     Float=3.3
    )) {isSetParameters=true, module=SetVariables_#4}
   &InitialMeshOrientation(
    Orientation Axis=<0 0 1>
    Orientation Vector:NiagaraPosition=$Particles.Position
    Rotation=<0 0.05 0>
    Use Orientation Vector=true
    Use Rotation Vector=true
   ) {module=InitialMeshOrientation001}
   &DynamicMaterialParameters(
    Index 0 Param 1=&RandomRangeFloat(Minimum=0, Maximum=1)
    Module.Param0WriteEnabled=true
    Module.Param1WriteEnabled=true
    Module.Param2WriteEnabled=true
    Module.Param3WriteEnabled=true
   )
   &KillParticles(Kill Particles=&Boolean_AndOperation(
     Bool1=&SetBoolByFloatComparison(
      A=&Frac_Float(Float=&Multiply_Float(A=$Emitter.Age, B=$System.TurbineSpinSpeed))
      B=0.44
     )
     Bool2=&SetBoolByFloatComparison(
      A=&Frac_Float(Float=&Multiply_Float(A=$Emitter.Age, B=$System.TurbineSpinSpeed))
      B=0.5
      Comparison Type=::NewEnumerator6
     )
    ))
  ]
  particleUpdateScript [
   &ParticleState()
   &ScaleColor(
    Scale Alpha=&FloatFromCurve(FloatCurve=DataInterface() {propertyValues {Curve {keys [{arriveTangent=-2.3614, interpMode=RCIM_Cubic, leaveTangent=-2.3614, tangentMode=RCTM_User, time=0, value=1}, <at 1 0 RCIM_Cubic>]}}})
    ScaleA=true
    ScaleRGB=true
   )
   &ScaleMeshSize(Scale Factor=&VectorFromFloat(Value=&FloatFromCurve(FloatCurve=DataInterface() {propertyValues {Curve {keys [{arriveTangent=-1.9281, interpMode=RCIM_Cubic, leaveTangent=-1.9281, tangentMode=RCTM_User, time=0, value=1}, <at 1 0>]}}})))
  ]
  rendererClasses [&NiagaraSpriteRendererProperties, &NiagaraMeshRendererProperties]
  renderers [
   {rendererClass=&NiagaraSpriteRendererProperties}
   {rendererClass=&NiagaraMeshRendererProperties, rendererIndex=1}
   {rendererClass=&NiagaraMeshRendererProperties, rendererIndex=2}
  ]
 }
 MiddleFire {
  emitterUpdateScript [
   &EmitterState()
   &SpawnRate(SpawnRate=50, Use Spawn Probability=false)
  ]
  particleSpawnScript [
   &InitializeParticle(
    Color=<rgba 1 0.1 0 1>
    Lifetime=0.5
    Mesh Scale=<0.1 0.06 0.16>
    Mesh Scale Mode=::NewEnumerator4
    Sprite Size Mode=::NewEnumerator3
    Uniform Sprite Size=12
    UsePositionOffset=false
    Write Material Random=true
   )
   /SetVariables(Transient.SpinSegments=16) {isSetParameters=true, module=SetVariables_#5}
   /SetVariables(Transient.SpinLocation:NiagaraFloat=&Divide_Float(
     A=&Ceil_Float(Float=&Multiply_Float(
       A=&Frac_Float(Float=&Multiply_Float(A=$Emitter.Age, B=$System.TurbineSpinSpeed))
       B=$Transient.SpinSegments
      ))
     B=$Transient.SpinSegments
    )) {isSetParameters=true, module=SetVariables_#6}
   &SkeletalMeshLocation(
    Inherit Velocity=false
    MaskSpawnGroup=false
    Mesh Sampling Type=::NewEnumerator5
    Skeletal Mesh=$User.SkeletalMesh
    Socket Sampling Mode=::NewEnumerator5
   )
   /SetVariables() {isSetParameters=true, module=SetVariables_#7}
   /SetVariables(Particles.DEBUG_MeshSize:Vector3f=&VectorFromFloat(Value=0.25)) {isSetParameters=true, module=SetVariables_#8}
   /SetVariables(Particles.Position=&Multiply_VectorByFloat(
     Vector=&MakeVector(
      X=0.3
      Y=&Sine(Normalized Angle=$Transient.SpinLocation, Period=-1, Scale=1, Bias=0)
      Z=&Cosine(Normalized Angle=$Transient.SpinLocation, Period=-1, Scale=1, Bias=0)
     )
     Float=4.25
    )) {isSetParameters=true, module=SetVariables_#4}
   &InitialMeshOrientation(
    Orientation Axis=<0 0 1>
    Orientation Vector:NiagaraPosition=$Particles.Position
    Rotation=<0 0.05 0>
    Use Orientation Vector=true
    Use Rotation Vector=true
   ) {module=InitialMeshOrientation001}
   &DynamicMaterialParameters(
    Index 0 Param 1=&RandomRangeFloat(Minimum=0, Maximum=1)
    Module.Param0WriteEnabled=true
    Module.Param1WriteEnabled=true
    Module.Param2WriteEnabled=true
    Module.Param3WriteEnabled=true
   )
   &KillParticles(Kill Particles=&Boolean_AndOperation(
     Bool1=&SetBoolByFloatComparison(
      A=&Frac_Float(Float=&Multiply_Float(A=$Emitter.Age, B=$System.TurbineSpinSpeed))
      B=0.44
     )
     Bool2=&SetBoolByFloatComparison(
      A=&Frac_Float(Float=&Multiply_Float(A=$Emitter.Age, B=$System.TurbineSpinSpeed))
      B=0.5
      Comparison Type=::NewEnumerator6
     )
    ))
  ]
  particleUpdateScript [
   &ParticleState()
   &ScaleColor(
    Scale Alpha=&FloatFromCurve(FloatCurve=DataInterface() {propertyValues {Curve {keys [{arriveTangent=-2.3614, interpMode=RCIM_Cubic, leaveTangent=-2.3614, tangentMode=RCTM_User, time=0, value=1}, <at 1 0 RCIM_Cubic>]}}})
    ScaleA=true
    ScaleRGB=true
   )
   &ScaleMeshSize(Scale Factor=&VectorFromFloat(Value=&FloatFromCurve(FloatCurve=DataInterface() {propertyValues {Curve {keys [{arriveTangent=-1.9281, interpMode=RCIM_Cubic, leaveTangent=-1.9281, tangentMode=RCTM_User, time=0, value=1}, <at 1 0>]}}})))
   /SetVariables(
    Particles.Sprite_CameraOffset=2
    Particles.Sprite_Color:LinearColor=&MultiplyLinearColor(
     Linear Color=<rgba 6 0.8 0.1 0.3>
     Scale Factor=&MakeVector4_FromFloat(Float=&FloatFromCurve(FloatCurve=DataInterface() {propertyValues {Curve {keys [{arriveTangent=-1.7235, interpMode=RCIM_Cubic, leaveTangent=-1.7235, tangentMode=RCTM_User, time=0, value=1}, <at 1 0>]}}}))
    )
    Particles.Sprite_DynParam01:Vector4f=&MakeVector4(X=1, Y=0, Z=0, W=0)
    Particles.Sprite_DynParam02:Vector4f=&MakeVector4(X=0, Y=50, Z=0, W=2)
   ) {isSetParameters=true}
  ]
  rendererClasses [&NiagaraSpriteRendererProperties, &NiagaraMeshRendererProperties]
  renderers [
   {rendererClass=&NiagaraSpriteRendererProperties}
   {rendererClass=&NiagaraMeshRendererProperties, rendererIndex=1}
  ]
 }
 ThinFire {
  emitterUpdateScript [
   &EmitterState()
   &SpawnRate(SpawnRate=50, Use Spawn Probability=false)
  ]
  particleSpawnScript [
   /SetVariables(Transient.SpinSegments=16) {isSetParameters=true, module=SetVariables_#5}
   /SetVariables(Transient.SpinSpeed=2) {isSetParameters=true}
   /SetVariables(Transient.SpinLocation:NiagaraFloat=&Divide_Float(
     A=&Ceil_Float(Float=&Multiply_Float(
       A=&Frac_Float(Float=&Multiply_Float(A=$Emitter.Age, B=$Transient.SpinSpeed))
       B=$Transient.SpinSegments
      ))
     B=$Transient.SpinSegments
    )) {isSetParameters=true, module=SetVariables_#6}
   &SkeletalMeshLocation(
    Inherit Velocity=false
    MaskSpawnGroup=false
    Mesh Sampling Type=::NewEnumerator5
    Skeletal Mesh=$User.SkeletalMesh
    Socket Sampling Mode=::NewEnumerator5
   )
   &InitializeParticle(
    Color=<rgba 1 0.2 0 0.2>
    Lifetime=0.25
    Mesh Scale=<0.05 0.02 0.22>
    Mesh Scale Mode=::NewEnumerator4
    Sprite Size Mode=::NewEnumerator3
    UsePositionOffset=false
    Write Material Random=true
   )
   /SetVariables() {isSetParameters=true, module=SetVariables_#7}
   /SetVariables(Particles.DEBUG_MeshSize:Vector3f=&VectorFromFloat(Value=0.25)) {isSetParameters=true, module=SetVariables_#8}
   /SetVariables(Particles.Position=&Multiply_VectorByFloat(
     Vector=&MakeVector(
      X=0.2
      Y=&Sine(Normalized Angle=$Transient.SpinLocation, Period=-1, Scale=1, Bias=0.01)
      Z=&Cosine(Normalized Angle=$Transient.SpinLocation, Period=-1, Scale=1, Bias=0.01)
     )
     Float=4.25
    )) {isSetParameters=true, module=SetVariables_#4}
   &InitialMeshOrientation(
    Orientation Axis=<0 0 1>
    Orientation Vector:NiagaraPosition=$Particles.Position
    Rotation=<0 0.05 0>
    Use Orientation Vector=true
    Use Rotation Vector=true
   ) {module=InitialMeshOrientation001}
   &InitialMeshOrientation(
    Orientation Axis=<0 0 1>
    Orientation Vector=/Niagara/DynamicInputs/Vectors/RotateVector(
     Vector To Rotate:Vector3f=$Output.SphereLocation.SphereNormal
     Yaw:NiagaraFloat=0
     Pitch:NiagaraFloat=0
     Roll:NiagaraFloat=0
     Rotation Rate:NiagaraFloat=$Engine.DeltaTime
    )
    Use Orientation Vector=true
    Use Rotation Vector=true
   ) {enabled=false}
   &DynamicMaterialParameters(
    Index 0 Param 1=&RandomRangeFloat(Minimum=0, Maximum=1)
    Module.Param0WriteEnabled=true
    Module.Param1WriteEnabled=true
    Module.Param2WriteEnabled=true
    Module.Param3WriteEnabled=true
   )
   &KillParticles(Kill Particles=&Boolean_AndOperation(
     Bool1=&SetBoolByFloatComparison(
      A=&Frac_Float(Float=&Multiply_Float(A=$Emitter.Age, B=$Transient.SpinSpeed))
      B=0.44
     )
     Bool2=&SetBoolByFloatComparison(
      A=&Frac_Float(Float=&Multiply_Float(A=$Emitter.Age, B=$Transient.SpinSpeed))
      B=0.5
      Comparison Type=::NewEnumerator6
     )
    ))
  ]
  particleUpdateScript [
   &ParticleState()
   &ScaleColor(
    Scale Alpha=&FloatFromCurve(FloatCurve=DataInterface() {propertyValues {Curve {keys [{arriveTangent=-2.3614, interpMode=RCIM_Cubic, leaveTangent=-2.3614, tangentMode=RCTM_User, time=0, value=1}, <at 1 0 RCIM_Cubic>]}}})
    ScaleA=true
    ScaleRGB=true
   )
  ]
  rendererClasses [&NiagaraSpriteRendererProperties, &NiagaraMeshRendererProperties]
  renderers [
   {rendererClass=&NiagaraSpriteRendererProperties}
   {rendererClass=&NiagaraMeshRendererProperties, rendererIndex=1}
   {rendererClass=&NiagaraMeshRendererProperties, rendererIndex=2}
  ]
 }
 Turbine_Glow {
  emitterUpdateScript [
   &EmitterState(
    Life Cycle Mode=::NewEnumerator1
    Loop Behavior=::NewEnumerator1
    Loop Duration Mode=::NewEnumerator1
    Reset Age On Awaken=true
    Scalability Mode=::NewEnumerator1
    UseLoopDelay=false
   )
   &SpawnBurst_Instantaneous(Spawn Count=1, Use Spawn Probability=false)
  ]
  particleSpawnScript [
   &InitializeParticle(
    Color=<rgba 5 0.7083 0 0.017>
    Lifetime=0.3
    Sprite Size Mode=::NewEnumerator3
    Uniform Sprite Size=30
    UsePositionOffset=false
    Write Material Random=true
   )
   &SkeletalMeshLocation(
    Inherit Velocity=false
    MaskSpawnGroup=false
    Mesh Sampling Type=::NewEnumerator5
    Skeletal Mesh=$User.SkeletalMesh
    Socket Sampling Mode=::NewEnumerator5
   )
  ]
  particleUpdateScript [
   &ParticleState(Kill Particles When Lifetime Has Elapsed=false, Let Infinitely Lived Particles Die When Emitter Deactivates=true)
   /SetVariables(Transient.SpinSegments=16) {isSetParameters=true}
   /SetVariables(Transient.SpinLocation:NiagaraFloat=&Divide_Float(
     A=&Ceil_Float(Float=&Multiply_Float(
       A=&Frac_Float(Float=&Multiply_Float(A=$Emitter.Age, B=$System.TurbineSpinSpeed))
       B=$Transient.SpinSegments
      ))
     B=$Transient.SpinSegments
    )) {isSetParameters=true}
   /SetVariables(Particles.Position=&Multiply_VectorByFloat(
     Vector=&MakeVector(
      X=0.3
      Y=&Sine(Normalized Angle=$Transient.SpinLocation, Period=-1, Scale=0.45, Bias=0.65)
      Z=&Cosine(Normalized Angle=$Transient.SpinLocation, Period=-1, Scale=1, Bias=0)
     )
     Float=4.25
    )) {isSetParameters=true}
   &ScaleColor(ScaleA=false, ScaleRGB=false) {enabled=false}
   &ScaleSpriteSize(Scale Factor=&Vector2DFromFloat(Value=&FloatFromCurve(FloatCurve=DataInterface() {propertyValues {Curve {keys [<at 0 0.5>,<at 0.5 1>,<at 1 0.5>]}}}))) {enabled=false}
   &CameraOffset(Camera Offset Amount=10)
   &FN_Light_Attributes(
    Color=$Particles.Color
    Create Color Attribute=true
    Create Exponent Attribute=true
    Create Position Attribute=true
    Create Radius Attribute=true
    Exponent=12
    Position=&Add_Vector(A=Linked() {linkedVariable {name=Particles.Position, type=NiagaraPosition}}, B=<10 -5 0>)
    Radius=35
    Scale Color=5
   )
  ]
  rendererClasses [&NiagaraSpriteRendererProperties, &NiagaraLightRendererProperties]
  renderers [
   {rendererClass=&NiagaraSpriteRendererProperties}
   {rendererClass=&NiagaraLightRendererProperties, rendererIndex=1}
  ]
 }
}
summary {
 systemName=NS_RocketLauncher_JunkGun_Grinder
 emitters [
  {emitterName=MiddleFire, rendererClasses [&NiagaraSpriteRendererProperties, &NiagaraMeshRendererProperties]}
  {emitterName=ThinFire, rendererClasses [&NiagaraSpriteRendererProperties, &NiagaraMeshRendererProperties]}
  {emitterName=LongFire, rendererClasses [&NiagaraSpriteRendererProperties, &NiagaraMeshRendererProperties]}
  {bEnabled=false, emitterName=Empty003, rendererClasses [&NiagaraLightRendererProperties]}
  {emitterName=Hopper_Sparks, rendererClasses [&NiagaraSpriteRendererProperties]}
  {emitterName=Hopper_Glow, rendererClasses [&NiagaraSpriteRendererProperties, &NiagaraLightRendererProperties]}
  {emitterName=Turbine_Glow, rendererClasses [&NiagaraSpriteRendererProperties, &NiagaraLightRendererProperties]}
  {emitterName=Hopper_Debris, rendererClasses [&NiagaraMeshRendererProperties]}
 ]
 userVariables [
  User.bIsHitting {type=NiagaraBool}
  User.CollectingMaterial {type=NiagaraFloat, defaultValue=0, description=""}
  User.ImpactLocation {type=Vector3f, description="", defaultValue {struct=Vector3f, value=<-112.3821 66.9813 -14.57>}}
  User.RibbonIDOffset {type=NiagaraInt32}
  User.SkeletalMesh {
   type=NiagaraDataInterfaceSkeletalMesh
   defaultValue {
    struct=NiagaraExt_VariableValue_DataInterface
    value {dataInterface=&self:NiagaraDataInterfaceSkeletalMesh_0}
   }
  }
 ]
}
