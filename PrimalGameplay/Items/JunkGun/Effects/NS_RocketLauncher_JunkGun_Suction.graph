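# Niagara system for the JunkGun rocket launcher's suction effect, pulling smoke, speed lines and debris toward the gun.
# A ribbon tube and those particles follow Hermite splines from the impact point, or ahead of the gun without a hit, to a skeletal mesh sample.
# Debris spawns only while hitting, varying by CollectingMaterial.
# User.Underwater drives muzzle bubbles, plus drops and foam constrained to the water plane.
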NiagaraSystem /PrimalGameplay/Items/JunkGun/Effects/NS_RocketLauncher_JunkGun_Suction
dependencies {
 usedDataInterfaces [&NiagaraDataInterfaceCurve, &NiagaraDataInterfaceArrayFloat]
 usedDynamicInputs [
  &VectorLength
  &Subtract_Vector
  &Multiply_Float
 ]
 usedModules [
  &SystemState
  &self:SystemSpawnScript.SystemScriptSource.SystemScriptGraph.NiagaraNodeAssignment_0.SetVariables_#1
  &self:SystemSpawnScript.SystemScriptSource.SystemScriptGraph.NiagaraNodeAssignment_2.SetVariables_#2
  &self:SystemSpawnScript.SystemScriptSource.SystemScriptGraph.NiagaraNodeAssignment_1.SetVariables_#3
  &self:SystemSpawnScript.SystemScriptSource.SystemScriptGraph.NiagaraNodeAssignment_3.SetVariables_#4
  &self:SystemSpawnScript.SystemScriptSource.SystemScriptGraph.NiagaraNodeAssignment_4.SetVariables_#5
 ]
}
emitters {
 Bubbles_Muzzle {
  emitterUpdateScript [
   &EmitterState()
   &EmitterScalability(
    MaxDistance=2000
    Module.bApplySpawnCountScale=true
    Module.bApplySpawnCountScaleByDistanceFraction=true
    Module.bUseMaxDistance=true
    Module.bUseMinDistance=false
    SpawnCountScaleByDistanceFraction=DataInterface() {propertyValues {Curve {keys [<at 0 1>,<at 1 0.5>]}}}
   ) {enabled=false}
   &EmitterLifeCycle() {enabled=false}
   &SpawnRate(SpawnRate=&Lerp_Float(A=0, B=5, Alpha=$User.Underwater), Use Spawn Probability=false)
   &SpawnRate(SpawnRate=5, Use Spawn Probability=false) {enabled=false, module=SpawnRate001}
  ]
  particleSpawnScript [
   /Niagara/Modules/Spawn/Initialization/InitializeParticle(
    Lifetime=0.75
    Module.Write Color=true
    Module.Write Lifetime=true
    Module.Write Mass=true
    Module.Write Material Random=true
    Module.Write Position=true
    Module.Write Scale=false
    Module.Write SpriteRotation=false
    Module.Write SpriteSize=false
   )
   /Niagara/Modules/Spawn/Orientation/V2/InitialMeshOrientation(
    Orientation Vector=/Niagara/DynamicInputs/Random/RandomVector(
     Vector Scale:NiagaraFloat=1
     Constraint Vector:Vector3f=<0 0 1>
     Randomness Mode:ENiagaraRandomnessMode=::NewEnumerator0
     Random Seed:NiagaraInt32=0
     Fixed Random Seed:NiagaraBool=false
    )
    Use Orientation Vector=true
    Use Rotation Vector=false
   )
   &AddVelocity(Velocity=&RotateVectorByQuaternion(Vector To Rotate=<0 125 -125>, Quaternion=$Output.SkeletalMeshLocation.SampledRotation)) {enabled=false}
   &SkeletalMeshLocation(
    Apply Sampled Orientation to Mesh Particles=true
    Inherit Velocity=false
    MaskSpawnGroup=false
    Mesh Sampling Type=::NewEnumerator5
    Sampled Position Offset=<300 0 0>
    Sampled Position Offset Coordinate Space:ENiagaraMeshTransforms=::NewEnumerator3
    Skeletal Mesh=$User.Skeletal Mesh
    Socket Sampling Mode=::NewEnumerator5
   ) {module=SkeletalMeshLocation001}
  ]
  particleUpdateScript [
   &ParticleState()
   &SkeletalMeshLocation(
    Apply Sampled Orientation to Mesh Particles=true
    Apply To Particle Position=false
    Inherit Velocity=false
    MaskSpawnGroup=false
    Mesh Sampling Type=::NewEnumerator5
    Sampled Position Offset Coordinate Space:ENiagaraMeshTransforms=::NewEnumerator3
    Skeletal Mesh=$User.Skeletal Mesh
    Socket Sampling Mode=::NewEnumerator5
   ) {module=SkeletalMeshLocation002}
   /SetVariables(Particles.Age:NiagaraFloat=&Lerp_Float(A=0, B=$Particles.Age, Alpha=$User.Underwater)) {enabled=false, isSetParameters=true}
   &SkeletalMeshLocation(
    Apply Sampled Orientation to Mesh Particles=true
    Inherit Velocity=false
    MaskSpawnGroup=false
    Mesh Sampling Type=::NewEnumerator5
    Sampled Position Offset=&MakeVector(
     X=0
     Y=&Lerp_Float(
      A=75
      B=0
      Alpha=&FloatFromCurve(FloatCurve=DataInterface() {propertyValues {Curve {keys [
           {arriveTangent=0.9687, interpMode=RCIM_Cubic, leaveTangent=0.9687, tangentMode=RCTM_User, time=0, value=0}
           {arriveTangent=1.4069, interpMode=RCIM_Cubic, leaveTangent=1.4069, tangentMode=RCTM_User, time=1, value=1}
          ]}}})
     )
     Z=0
    )
    Sampled Position Offset Coordinate Space:ENiagaraMeshTransforms=::NewEnumerator3
    Skeletal Mesh=$User.Skeletal Mesh
    Socket Sampling Mode=::NewEnumerator5
   ) {enabled=false}
   &Drag(Use Linear Drag=true, Use Rotational Drag=false) {enabled=false}
   &GravityForce(Gravity=<0 0 75>) {enabled=false}
   &PointAttractionForce(
    Attraction Radius=320
    Attraction Strength=80
    Attractor Position:NiagaraPosition=$Output.SkeletalMeshLocation002.SampledPosition
    Kill Within Radius=false
    Module.Use Falloff=true
    UseAttractorOffset=false
   )
   &SolveForcesAndVelocity(Clamp Velocity=false, Limit Acceleration=false)
   &DynamicMaterialParameters(
    Index 0 Param 1=&FloatFromCurve(
     FloatCurve=DataInterface() {propertyValues {Curve {keys [
         {arriveTangent=0.7622, interpMode=RCIM_Cubic, leaveTangent=0.7622, tangentMode=RCTM_User, time=-0.0046, value=0}
         {arriveTangent=6.275, interpMode=RCIM_Cubic, leaveTangent=6.275, tangentMode=RCTM_User, time=1.0011, value=3.2}
        ]}}}
     Scale Curve=0.9
    )
    Index 0 Param 2=&FloatFromCurve(FloatCurve=DataInterface() {propertyValues {Curve {keys [<at 0.1 1>,<at 1.0011 5>]}}}, Scale Curve=0.4)
    Index 0 Param 3=&FloatFromCurve(FloatCurve=DataInterface() {propertyValues {Curve {keys [<at -0.0046 70>,<at 1.0048 15>]}}})
    Index 0 Param 4=HlslExpression() {hlslExpression="rand(5.0) + 4"}
    Module.Param0WriteEnabled=true
    Module.Param1WriteEnabled=true
    Module.Param2WriteEnabled=true
    Module.Param3WriteEnabled=true
   )
   &DynamicMaterialParameters(
    Index 0 Param 1=&FloatFromCurve(
     FloatCurve=DataInterface() {propertyValues {Curve {keys [{arriveTangent=-10.0518, interpMode=RCIM_Cubic, leaveTangent=-6.7075, tangentMode=RCTM_User, time=-0.0046, value=3.2841}, <at 1.0011 0.002>]}}}
     Scale Curve=0.9
    )
    Index 0 Param 2=&FloatFromCurve(FloatCurve=DataInterface() {propertyValues {Curve {keys [<at 0.1 5>,<at 1.0011 1>]}}}, Scale Curve=0.4)
    Index 0 Param 3=&FloatFromCurve(FloatCurve=DataInterface() {propertyValues {Curve {keys [<at -0.0046 60.8837>,<at 1.0048 166.0627>]}}}, Scale Curve=0.1)
    Index 0 Param 4=HlslExpression() {hlslExpression="rand(4.0) + 4"}
    Module.Param0WriteEnabled=true
    Module.Param1WriteEnabled=true
    Module.Param2WriteEnabled=true
    Module.Param3WriteEnabled=true
   ) {enabled=false, module=DynamicMaterialParameters001}
   &ScaleColor(
    Scale Alpha=&FloatFromCurve(FloatCurve=DataInterface() {propertyValues {Curve {keys [<at 0 -1>,<at 0.05 0>,<at 0.7904 0.0132>,<at 1.0011 -0.9961>]}}})
    ScaleA=true
    ScaleRGB=true
   )
   &CameraOffset(Camera Offset Amount=50) {enabled=false}
   &KillParticlesInVolume(
    Kill Shape=::NewEnumerator3
    Slab Width=35
    Volume Origin:Vector3f=&MakeVector(
     X=0
     Y=0
     Z=&Lerp_Float(
      A=&MakeFloatFromVector(Vector=$Output.SkeletalMeshLocation.SampledPosition, Channel=::NewEnumerator2)
      B=&MakeFloatFromVector(Vector=$User.WaterPlaneLocation, Channel=::NewEnumerator2)
      Alpha=$User.Underwater
     )
    )
   ) {enabled=false}
   &KillParticles(Kill Particles=&SetBoolByFloatComparison(
     A=&Subtract_Float(
      A=&MakeFloatFromVector(Vector=Linked() {linkedVariable {name=Particles.Initial.Position, type=NiagaraPosition}}, Channel=::NewEnumerator2)
      B=-25
     )
     B=&MakeFloatFromVector(Vector=$Output.SkeletalMeshLocation002.SampledPosition, Channel=::NewEnumerator2)
    ))
  ]
  rendererClasses [&NiagaraMeshRendererProperties]
  renderers [{rendererClass=&NiagaraMeshRendererProperties, rendererIndex=0}]
 }
 Debris3D001 {
  emitterUpdateScript [
   &EmitterState()
   &SpawnRate(
    Spawn Probability=&MakeCustomFloatFromBool(Bool=$User.bIsHitting, True Float=1, False Float=0)
    SpawnRate=5
    Use Spawn Probability=true
   )
  ]
  particleSpawnScript [
   /SetVariables(Particles.MeshIndex=&Add_Integer(A=&MakeIntFromFloat(Float=$User.CollectingMaterial), B=-1)) {enabled=false, isSetParameters=true}
   /SetVariables(Particles.VisibilityTag=&MakeIntFromFloat(Float=$User.CollectingMaterial)) {isSetParameters=true}
   /SetVariables(Particles.MeshIndex=2) {enabled=false, isSetParameters=true}
   &InitializeParticle(
    Color=&RandomRangeLinearColor(Minimum=<rgba 0.4184 0.3349 0.3189 1>, Maximum=<rgba 0.9809 0.9598 0.877 1>)
    Lifetime=&RandomRangeFloat(Minimum=0.4, Maximum=1)
    Mesh Scale=&VectorFromFloat(Value=&RandomRangeFloat(Minimum=0.6, Maximum=1.2))
    Module.Write Color=true
    Module.Write Lifetime=true
    Module.Write Mass=true
    Module.Write Material Random=true
    Module.Write Position=false
    Module.Write Scale=true
    Module.Write SpriteRotation=false
    Module.Write SpriteSize=false
   )
   /SetVariables(Particles.RandomOffset:Vector3f=&RandomRangeVector(Minimum=<-75 -75 -75>, Maximum=<75 75 75>)) {isSetParameters=true, module=SetVariables_#6}
   /Niagara/Modules/Spawn/Orientation/V3/InitialMeshOrientation(
    Orientation Axis=&RandomRangeVector(Minimum=<-1 -1 -1>, Maximum=<1 1 1>)
    Use Orientation Vector=true
    Use Rotation Vector=false
   )
  ]
  particleUpdateScript [
   &ParticleState()
   &SkeletalMeshLocation(
    Apply To Particle Position=false
    Inherit Velocity=false
    MaskSpawnGroup=false
    Mesh Sampling Type=::NewEnumerator5
    Sampled Position Offset Coordinate Space:ENiagaraMeshTransforms=::NewEnumerator3
    Skeletal Mesh=$User.Skeletal Mesh
    Socket Sampling Mode=::NewEnumerator5
   )
   &UpdateMeshOrientation(Rotation Rate=2, Rotation Vector=<1 1 0>)
   /Game/Effects/Niagara/Modules/Curves/FN_HermiteCubicSpline_v1/FN_HermiteCubicSpline(
    Compute Direction=true
    Direction End:Vector3f=&MakeCustomVectorFromBool(
     Bool=$User.bIsHitting
     True Vector=&Multiply_VectorByFloat(
      Vector=&RotateVectorByQuaternion(Vector To Rotate=<1 0 0>, Quaternion=$Output.SkeletalMeshLocation.SampledRotation)
      Float=$System.TangentLength
     )
     False Vector=<1 0 0>
    )
    Direction Start:Vector3f=&MakeCustomVectorFromBool(
     Bool=$User.bIsHitting
     True Vector=&Multiply_VectorByFloat(Vector=$User.Impact_Orientation, Float=$System.TangentLength)
     False Vector=<1 0 0>
    )
    Position Blend:NiagaraFloat=$Particles.NormalizedAge
    Position End:Vector3f=&Add_Vector(
     A=$Output.SkeletalMeshLocation.SampledPosition
     B=&RotateVectorByQuaternion(Vector To Rotate=<-12.5 0 0>, Quaternion=$Output.SkeletalMeshLocation.SampledRotation)
    )
    Position Start:Vector3f=&Add_Vector(
     A=&MakeCustomVectorFromBool(
      Bool=$User.bIsHitting
      True Vector=$User.ImpactLocation
      False Vector=/Niagara/DynamicInputs/Transforms/TransformPosition(Position=<150 0 0>, Source Space=::Local, Destination Space=::World)
     )
     B=$Particles.RandomOffset
    )
    Write Particles Position=true
    Write Particles Velocity=true
   )
   &FN_HermiteCubicSpline(
    Compute Direction=true
    Position Blend:NiagaraFloat=$Particles.NormalizedAge
    Position End:Vector3f=$Output.SkeletalMeshLocation.SampledPosition
    Position Start:Vector3f=&Add_Vector(
     A=&self:SmoothInterpolateParticleVector(
      Initial Vector:Vector3f=$System.PosOnFrontOfGun
      Vector to Interpolate:Vector3f=&MakeCustomVectorFromBool(Bool=$User.bIsHitting, True Vector=$User.ImpactLocation, False Vector=$System.PosOnFrontOfGun)
      Rate Of Change:NiagaraFloat=$System.PositionInterpolateSpeed
     )
     B=$Particles.RandomOffset
    )
    Write Particles Position=true
    Write Particles Velocity=true
   ) {enabled=false, module=FN_HermiteCubicSpline001}
   &ScaleMeshSize(Scale Factor=&VectorFromFloat(Value=&FloatFromCurve(FloatCurve=DataInterface() {propertyValues {Curve {keys [
          {arriveTangent=2.8442, interpMode=RCIM_Cubic, leaveTangent=2.8442, tangentMode=RCTM_User, time=0, value=0}
          <at 0.5 1 RCIM_Cubic>
          {arriveTangent=-2.3002, interpMode=RCIM_Cubic, leaveTangent=-2.3002, tangentMode=RCTM_User, time=1, value=0}
         ]}}})))
  ]
  rendererClasses [&NiagaraMeshRendererProperties]
  renderers [
   {rendererClass=&NiagaraMeshRendererProperties, rendererIndex=0}
   {rendererClass=&NiagaraMeshRendererProperties, rendererIndex=1}
   {rendererClass=&NiagaraMeshRendererProperties, rendererIndex=2}
  ]
 }
 DebrisSprite {
  emitterUpdateScript [
   &EmitterState()
   &SpawnRate(
    Spawn Probability=&MakeCustomFloatFromBool(Bool=$User.bIsHitting, True Float=1, False Float=0)
    SpawnRate=8
    Use Spawn Probability=true
   )
  ]
  particleSpawnScript [
   /Niagara/Modules/Spawn/Initialization/V2/InitializeParticle(
    Color Maximum=<rgba 0.9809 0.9598 0.877 1>
    Color Minimum=<rgba 0.4184 0.3263 0.3087 1>
    Color Mode=::NewEnumerator2
    Lifetime Max=1
    Lifetime Min=0.4
    Lifetime Mode=::NewEnumerator1
    Mesh Scale Mode=::NewEnumerator4
    Sprite Rotation Mode=::NewEnumerator2
    Sprite Size=&MakeVector2D(
     X=&SelectFloatFromArray(
      Direct Array Index=&MakeIntFromFloat(Float=&Subtract_Float(A=$User.CollectingMaterial, B=1))
      Float Selection Array=DataInterface() {propertyValues {FloatData [2.5,7,10]}}
     )
     Y=&SelectFloatFromArray(
      Direct Array Index=&MakeIntFromFloat(Float=&Subtract_Float(A=$User.CollectingMaterial, B=1))
      Float Selection Array=DataInterface() {propertyValues {FloatData [15,7,10]}}
     )
    )
    Sprite Size Mode=::NewEnumerator4
    Sprite UV Mode=::NewEnumerator3
    UsePositionOffset=false
    Write Material Random=true
   )
   /SetVariables(Particles.RandomOffset:Vector3f=&RandomRangeVector(Minimum=<-75 -75 -75>, Maximum=<75 75 75>)) {isSetParameters=true}
   /SetVariables(Particles.VisibilityTag=&MakeIntFromFloat(Float=$User.CollectingMaterial)) {isSetParameters=true}
   &DynamicMaterialParameters(
    Index 0 Param 1=&RandomRangeFloat(Minimum=0, Maximum=1)
    Module.Param0WriteEnabled=true
    Module.Param1WriteEnabled=true
    Module.Param2WriteEnabled=true
    Module.Param3WriteEnabled=true
   )
  ]
  particleUpdateScript [
   &ParticleState()
   &SkeletalMeshLocation(
    Apply To Particle Position=false
    Inherit Velocity=false
    MaskSpawnGroup=false
    Mesh Sampling Type=::NewEnumerator5
    Sampled Position Offset Coordinate Space:ENiagaraMeshTransforms=::NewEnumerator3
    Skeletal Mesh=$User.Skeletal Mesh
    Socket Sampling Mode=::NewEnumerator5
   )
   &FN_HermiteCubicSpline(
    Compute Direction=true
    Direction End:Vector3f=&MakeCustomVectorFromBool(
     Bool=$User.bIsHitting
     True Vector=&Multiply_VectorByFloat(
      Vector=&RotateVectorByQuaternion(Vector To Rotate=<1 0 0>, Quaternion=$Output.SkeletalMeshLocation.SampledRotation)
      Float=$System.TangentLength
     )
     False Vector=<1 0 0>
    )
    Direction Start:Vector3f=&MakeCustomVectorFromBool(
     Bool=$User.bIsHitting
     True Vector=&Multiply_VectorByFloat(Vector=$User.Impact_Orientation, Float=$System.TangentLength)
     False Vector=<1 0 0>
    )
    Position Blend:NiagaraFloat=$Particles.NormalizedAge
    Position End:Vector3f=&Add_Vector(
     A=$Output.SkeletalMeshLocation.SampledPosition
     B=&RotateVectorByQuaternion(Vector To Rotate=<-12.5 0 0>, Quaternion=$Output.SkeletalMeshLocation.SampledRotation)
    )
    Position Start:Vector3f=&Add_Vector(
     A=&MakeCustomVectorFromBool(
      Bool=$User.bIsHitting
      True Vector=$User.ImpactLocation
      False Vector=&TransformPosition(Position=<150 0 0>, Source Space=::Local, Destination Space=::World)
     )
     B=$Particles.RandomOffset
    )
    Write Particles Position=true
    Write Particles Velocity=true
   )
   &ScaleSpriteSize(Scale Factor=&Vector2DFromFloat(Value=&FloatFromCurve(FloatCurve=DataInterface() {propertyValues {Curve {keys [
          {arriveTangent=2.8442, interpMode=RCIM_Cubic, leaveTangent=2.8442, tangentMode=RCTM_User, time=0, value=0}
          <at 0.5 1 RCIM_Cubic>
          {arriveTangent=-2.3002, interpMode=RCIM_Cubic, leaveTangent=-2.3002, tangentMode=RCTM_User, time=1, value=0}
         ]}}})))
   &SpriteRotationRate(Rotation Rate=&RandomRangeFloat(Minimum=-1000, Maximum=1000))
   &ScaleColor(
    Scale RGB=&VectorFromFloat(Value=&SelectFloatFromArray(
      Direct Array Index=&MakeIntFromFloat(Float=&Subtract_Float(A=$User.CollectingMaterial, B=1))
      Float Selection Array=DataInterface() {propertyValues {FloatData [0.75,0.5,1]}}
     ))
    ScaleA=true
    ScaleRGB=true
   )
  ]
  rendererClasses [&NiagaraSpriteRendererProperties]
  renderers [
   {rendererClass=&NiagaraSpriteRendererProperties}
   {rendererClass=&NiagaraSpriteRendererProperties, rendererIndex=1}
   {rendererClass=&NiagaraSpriteRendererProperties, rendererIndex=2}
  ]
 }
 Drops_SnapToSurface_Detail {
  emitterUpdateScript [
   &EmitterState()
   &EmitterScalability(
    Module.bApplySpawnCountScale=true
    Module.bApplySpawnCountScaleByDistanceFraction=true
    Module.bUseMaxDistance=true
    Module.bUseMinDistance=true
    SpawnCountScaleByDistanceFraction=DataInterface() {propertyValues {Curve {keys [<at 0 1>,<at 1 0.3>]}}}
   ) {enabled=false}
   &SpawnBurst_Instantaneous(Use Spawn Probability=false) {enabled=false}
   &EmitterLifeCycle() {enabled=false}
   &SpawnRate(SpawnRate=&Lerp_Float(A=0, B=10, Alpha=$User.Underwater), Use Spawn Probability=false)
  ]
  particleSpawnScript [
   /SetVariables(Particles.Random=&UniformRangedFloat(Minimum=0, Maximum=1), Particles.SpriteFacing=<0 0 1>) {isSetParameters=true}
   /Niagara/Modules/Spawn/Initialization/InitializeParticle(
    Color=<rgba 1 1 1 0.8>
    Lifetime=&UniformRangedFloat(Minimum=1, Maximum=1.5)
    Mass=&UniformRangedFloat(Minimum=0.6, Maximum=1.2)
    Mesh Scale=<100 100 100>
    Module.Write Color=true
    Module.Write Lifetime=true
    Module.Write Mass=true
    Module.Write Material Random=true
    Module.Write Position=false
    Module.Write Scale=false
    Module.Write SpriteRotation=true
    Module.Write SpriteSize=true
    Position=&SimulationPosition()
    Sprite Rotation=&NormalizedAngleToDegrees(NormalizedAngle=&UniformRangedFloat(Minimum=0, Maximum=360))
    Sprite Size=&Vector2DFromFloat(Value=&RandomRangeFloat(Minimum=2.5, Maximum=5))
   )
   /Niagara/Modules/Spawn/Orientation/V2/InitialMeshOrientation(
    Orientation Vector=<0 0 0>
    Rotation=&UniformRangedVector(Minimum=<0 0 0>, Maximum=<0 0 1>)
    Use Orientation Vector=true
    Use Rotation Vector=true
   ) {enabled=false}
   &DynamicMaterialParameters(
    Index 0 Param 1=1
    Module.Param0WriteEnabled=true
    Module.Param1WriteEnabled=false
    Module.Param2WriteEnabled=false
    Module.Param3WriteEnabled=false
   )
   &SkeletalMeshLocation(
    Inherit Velocity=false
    MaskSpawnGroup=false
    Mesh Sampling Type=::NewEnumerator5
    Sampled Position Offset=<100 0 0>
    Sampled Position Offset Coordinate Space:ENiagaraMeshTransforms=::NewEnumerator3
    Skeletal Mesh=$User.Skeletal Mesh
    Socket Sampling Mode=::NewEnumerator5
   )
   &CylinderLocation(
    Cylinder Height=0
    Cylinder Radius=40
    Lathe Profile=DataInterface() {propertyValues {Curve {keys []}}}
    MaskSpawnGroup=false
    Module.Enable Lathe Profile=false
    Module.Override Randomness=false
    Module.Surface Only=true
    Use Endcaps In Surface Only Mode=false
   )
   &InitializeWaterSnapToOwner()
  ]
  particleUpdateScript [
   &ParticleState()
   &ScaleColor(
    Scale Alpha=&FloatFromCurve(FloatCurve=DataInterface() {propertyValues {Curve {keys [
         <at 0 0 RCIM_Cubic>
         {arriveTangent=0.002, interpMode=RCIM_Cubic, leaveTangent=0.002, time=0.1183, value=1.0113}
         <at 1.0011 0.002 RCIM_Cubic>
        ]}}})
    ScaleA=true
    ScaleRGB=true
   )
   &PointForce(Force Falloff Distance=0, Force Strength=180, Module.Use Falloff=false, Module.UseLinearForce=false)
   &AccelerationForce(Acceleration=&UniformRangedVector(Minimum=<-500 -500 0>, Maximum=<500 500 0>)) {enabled=false}
   &Drag(Drag=2, Use Linear Drag=true, Use Rotational Drag=false)
   &SolveForcesAndVelocity(Clamp Velocity=false, Limit Acceleration=false)
   &ScaleMeshSize() {enabled=false, module=MeshSizeScale}
   &ScaleSpriteSize(Scale Factor=&Vector2DFromFloat(Value=&FloatFromCurve(FloatCurve=DataInterface() {propertyValues {Curve {keys [
          <at 0 0 RCIM_Cubic>
          {arriveTangent=1.4535, interpMode=RCIM_Cubic, leaveTangent=1.4535, time=0.2, value=1.25}
          {arriveTangent=-1.5578, interpMode=RCIM_Cubic, leaveTangent=-1.5578, time=0.6958, value=1.0113}
          <at 1.0011 0.002 RCIM_Cubic>
         ]}}})))
   &ConstrainPositionToPlane(Plane Normal:Vector3f=$Particles.WaterPlaneNormal, Plane Position:Vector3f=$Particles.WaterPlaneLocation)
  ]
  rendererClasses [&NiagaraSpriteRendererProperties, &NiagaraMeshRendererProperties]
  renderers [
   {rendererClass=&NiagaraSpriteRendererProperties}
   {rendererClass=&NiagaraMeshRendererProperties, rendererIndex=1}
  ]
 }
 Smoke {
  emitterUpdateScript [
   &EmitterState()
   &SpawnRate(SpawnRate=15, Use Spawn Probability=false)
  ]
  particleSpawnScript [
   &InitializeParticle(
    Color=<rgba 0.283 0.2149 0.1626 0.5>
    Lifetime=&RandomRangeFloat(Minimum=0.5, Maximum=0.8)
    Module.Write Color=true
    Module.Write Lifetime=true
    Module.Write Mass=true
    Module.Write Material Random=true
    Module.Write Position=false
    Module.Write Scale=false
    Module.Write SpriteRotation=false
    Module.Write SpriteSize=true
    Sprite Rotation=&RandomRangeFloat(Minimum=0, Maximum=180)
    Sprite Size=<150 300>
   )
   /SetVariables(Particles.RandomOffset:Vector3f=&RandomRangeVector(Minimum=<-75 -75 -75>, Maximum=<75 75 75>)) {isSetParameters=true, module=SetVariables_#6}
   /SetVariables(Particles.RandomFrameStart:NiagaraInt32=&RandomRangeInt(Minimum=0, Maximum=63)) {enabled=false, isSetParameters=true}
  ]
  particleUpdateScript [
   &ParticleState()
   &SkeletalMeshLocation(
    Apply To Particle Position=false
    Inherit Velocity=false
    MaskSpawnGroup=false
    Mesh Sampling Type=::NewEnumerator5
    Sampled Position Offset Coordinate Space:ENiagaraMeshTransforms=::NewEnumerator3
    Skeletal Mesh=$User.Skeletal Mesh
    Socket Sampling Mode=::NewEnumerator5
   )
   &FN_HermiteCubicSpline(
    Compute Direction=true
    Direction End:Vector3f=&MakeCustomVectorFromBool(
     Bool=$User.bIsHitting
     True Vector=&Multiply_VectorByFloat(
      Vector=&RotateVectorByQuaternion(Vector To Rotate=<1 0 0>, Quaternion=$Output.SkeletalMeshLocation.SampledRotation)
      Float=$System.TangentLength
     )
     False Vector=<1 0 0>
    )
    Direction Start:Vector3f=&MakeCustomVectorFromBool(
     Bool=$User.bIsHitting
     True Vector=&Multiply_VectorByFloat(Vector=$User.Impact_Orientation, Float=$System.TangentLength)
     False Vector=<1 0 0>
    )
    Position Blend:NiagaraFloat=$Particles.NormalizedAge
    Position End:Vector3f=&Add_Vector(
     A=$Output.SkeletalMeshLocation.SampledPosition
     B=&RotateVectorByQuaternion(Vector To Rotate=<-12.5 0 0>, Quaternion=$Output.SkeletalMeshLocation.SampledRotation)
    )
    Position Start:Vector3f=&Add_Vector(
     A=&MakeCustomVectorFromBool(
      Bool=$User.bIsHitting
      True Vector=$User.ImpactLocation
      False Vector=&TransformPosition(Position=<150 0 0>, Source Space=::Local, Destination Space=::World)
     )
     B=$Particles.RandomOffset
    )
    Write Particles Position=true
    Write Particles Velocity=true
   )
   &FN_HermiteCubicSpline(
    Compute Direction=true
    Position Blend:NiagaraFloat=$Particles.NormalizedAge
    Position End:Vector3f=$Output.SkeletalMeshLocation.SampledPosition
    Position Start:Vector3f=&Add_Vector(
     A=&self:SmoothInterpolateParticleVector(
      Initial Vector:Vector3f=$System.PosOnFrontOfGun
      Vector to Interpolate:Vector3f=&MakeCustomVectorFromBool(Bool=$User.bIsHitting, True Vector=$User.ImpactLocation, False Vector=$System.PosOnFrontOfGun)
      Rate Of Change:NiagaraFloat=$System.PositionInterpolateSpeed
     )
     B=$Particles.RandomOffset
    )
    Write Particles Position=true
    Write Particles Velocity=true
   ) {enabled=false, module=FN_HermiteCubicSpline001}
   &ScaleColor(
    Scale Alpha=&FloatFromCurve(FloatCurve=DataInterface() {propertyValues {Curve {keys [<at 0 0 RCIM_Cubic>,<at 0.6 0.9 RCIM_Cubic>,<at 1 0.5 RCIM_Cubic>]}}})
    ScaleA=true
    ScaleRGB=true
   )
   &ScaleSpriteSize(Scale Factor=&Vector2DFromFloat(Value=&FloatFromCurve(FloatCurve=DataInterface() {propertyValues {Curve {keys [
          {arriveTangent=-1.506, interpMode=RCIM_Cubic, leaveTangent=-1.506, tangentMode=RCTM_User, time=0, value=1}
          {arriveTangent=-0.6366, interpMode=RCIM_Cubic, leaveTangent=-0.6366, tangentMode=RCTM_User, time=1, value=0}
         ]}}})))
   &SubUVAnimation(
    End Frame:NiagaraInt32=&Add_Integer(A=$Particles.RandomFrameStart, B=40)
    Start Frame:NiagaraInt32=$Particles.RandomFrameStart
   ) {enabled=false}
   &DynamicMaterialParameters(
    Index 0 Param 1=&FloatFromCurve(FloatCurve=DataInterface() {propertyValues {Curve {keys [<at 0 -0.5>,<at 1 0.25>]}}})
    Module.Param0WriteEnabled=true
    Module.Param1WriteEnabled=true
    Module.Param2WriteEnabled=true
    Module.Param3WriteEnabled=true
   ) {enabled=false}
   &DynamicMaterialParameters(
    Index 0 Param 1=&FloatFromCurve(FloatCurve=DataInterface() {propertyValues {Curve {keys [<at 0 1>,<at 1 0>]}}}, Scale Curve=0.5)
    Index 2 Param 1=&FloatFromCurve(FloatCurve=DataInterface() {propertyValues {Curve {keys [
         {arriveTangent=-10.5859, interpMode=RCIM_Cubic, leaveTangent=-10.5859, tangentMode=RCTM_User, time=0, value=2}
         <at 0.2 1 RCIM_Cubic>
         <at 1 2 RCIM_Cubic>
        ]}}})
    Index 2 Param 2=&FloatFromCurve(FloatCurve=DataInterface() {propertyValues {Curve {keys [<at 0 1>,<at 1 0>]}}}, Scale Curve=0.5)
    Index 2 Param 3=1
    Index 2 Param 4=&FloatFromCurve(FloatCurve=DataInterface() {propertyValues {Curve {keys [<at 0 1.2 RCIM_Cubic>,<at 0.2 0.8 RCIM_Cubic>,<at 1 1.2 RCIM_Cubic>]}}})
    Module.Param0WriteEnabled=true
    Module.Param0WriteEnabled2=true
    Module.Param1WriteEnabled=false
    Module.Param1WriteEnabled2=true
    Module.Param2WriteEnabled=true
    Module.Param2WriteEnabled2=true
    Module.Param3WriteEnabled=true
    Module.Param3WriteEnabled2=true
    Write Parameter Index 2=true
   ) {module=DynamicMaterialParameters001}
  ]
  rendererClasses [&NiagaraSpriteRendererProperties]
  renderers [
   {rendererClass=&NiagaraSpriteRendererProperties}
   {rendererClass=&NiagaraSpriteRendererProperties, rendererIndex=1}
  ]
 }
 SpeedLines {
  emitterUpdateScript [
   &EmitterState()
   &SpawnRate(SpawnRate=25, Use Spawn Probability=false)
  ]
  particleSpawnScript [
   &InitializeParticle(
    Color=<rgba 1 1 1 0.1>
    Lifetime=&RandomRangeFloat(Minimum=0.3, Maximum=0.5)
    Module.Write Color=true
    Module.Write Lifetime=true
    Module.Write Mass=true
    Module.Write Material Random=true
    Module.Write Position=false
    Module.Write Scale=false
    Module.Write SpriteRotation=false
    Module.Write SpriteSize=true
    Sprite Size=<3 60>
   )
   /SetVariables(Particles.RandomOffset:Vector3f=&RandomRangeVector(Minimum=<-75 -75 -75>, Maximum=<75 75 75>)) {isSetParameters=true, module=SetVariables_#6}
  ]
  particleUpdateScript [
   &ParticleState()
   &SkeletalMeshLocation(
    Apply To Particle Position=false
    Inherit Velocity=false
    MaskSpawnGroup=false
    Mesh Sampling Type=::NewEnumerator5
    Sampled Position Offset Coordinate Space:ENiagaraMeshTransforms=::NewEnumerator3
    Skeletal Mesh=$User.Skeletal Mesh
    Socket Sampling Mode=::NewEnumerator5
   )
   &FN_HermiteCubicSpline(
    Compute Direction=true
    Direction End:Vector3f=&MakeCustomVectorFromBool(
     Bool=$User.bIsHitting
     True Vector=&Multiply_VectorByFloat(
      Vector=&RotateVectorByQuaternion(Vector To Rotate=<1 0 0>, Quaternion=$Output.SkeletalMeshLocation.SampledRotation)
      Float=$System.TangentLength
     )
     False Vector=<1 0 0>
    )
    Direction Start:Vector3f=&MakeCustomVectorFromBool(
     Bool=$User.bIsHitting
     True Vector=&Multiply_VectorByFloat(Vector=$User.Impact_Orientation, Float=$System.TangentLength)
     False Vector=<1 0 0>
    )
    Position Blend:NiagaraFloat=$Particles.NormalizedAge
    Position End:Vector3f=&Add_Vector(
     A=$Output.SkeletalMeshLocation.SampledPosition
     B=&RotateVectorByQuaternion(Vector To Rotate=<-12.5 0 0>, Quaternion=$Output.SkeletalMeshLocation.SampledRotation)
    )
    Position Start:Vector3f=&Add_Vector(
     A=&MakeCustomVectorFromBool(
      Bool=$User.bIsHitting
      True Vector=$User.ImpactLocation
      False Vector=&TransformPosition(Position=<150 0 0>, Source Space=::Local, Destination Space=::World)
     )
     B=$Particles.RandomOffset
    )
    Write Particles Position=true
    Write Particles Velocity=true
   )
   &FN_HermiteCubicSpline(
    Compute Direction=true
    Position Blend:NiagaraFloat=$Particles.NormalizedAge
    Position End:Vector3f=$Output.SkeletalMeshLocation.SampledPosition
    Position Start:Vector3f=&Add_Vector(
     A=&self:SmoothInterpolateParticleVector(
      Initial Vector:Vector3f=$System.PosOnFrontOfGun
      Vector to Interpolate:Vector3f=&MakeCustomVectorFromBool(Bool=$User.bIsHitting, True Vector=$User.ImpactLocation, False Vector=$System.PosOnFrontOfGun)
      Rate Of Change:NiagaraFloat=$System.PositionInterpolateSpeed
     )
     B=$Particles.RandomOffset
    )
    Write Particles Position=true
    Write Particles Velocity=true
   ) {enabled=false, module=FN_HermiteCubicSpline001}
   &ScaleColor(
    Scale Alpha=&FloatFromCurve(FloatCurve=DataInterface() {propertyValues {Curve {keys [<at 0 0 RCIM_Cubic>,<at 0.5 1 RCIM_Cubic>,<at 1 0 RCIM_Cubic>]}}})
    ScaleA=true
    ScaleRGB=true
   )
   &ScaleSpriteSize(Scale Factor=&Vector2DFromFloat(Value=&FloatFromCurve(FloatCurve=DataInterface() {propertyValues {Curve {keys [<at 0 1>,<at 1 0>]}}})))
  ]
  rendererClasses [&NiagaraSpriteRendererProperties]
  renderers [{rendererClass=&NiagaraSpriteRendererProperties}]
 }
 TubeRibbon {
  emitterSpawnScript [/Game/Effects/Niagara/Modules/Partition/FN_PartitionAttributes(Number of Bone Sets=1, Particles Per Partition=5)]
  emitterUpdateScript [
   &EmitterState(Life Cycle Mode=::NewEnumerator1, Loop Behavior=::NewEnumerator1, Loop Duration Mode=::NewEnumerator1, UseLoopDelay=false)
   &SpawnBurst_Instantaneous(Spawn Count=$Emitter.FN_PartitionAttributes.Spawn Total, Use Spawn Probability=false)
  ]
  particleSpawnScript [
   &InitializeParticle(
    Module.Write Color=true
    Module.Write Lifetime=true
    Module.Write Mass=true
    Module.Write Material Random=true
    Module.Write Position=false
    Module.Write Scale=false
    Module.Write SpriteRotation=false
    Module.Write SpriteSize=true
    Sprite Size=<25 25>
   )
   /Game/Effects/Niagara/Modules/Partition/FN_Particle_partition_v5/FN_Particle_Partition_v5(Enable Ribbon ID=true)
   &InitializeRibbon(
    Module.Write Color=true
    Module.Write Lifetime=true
    Module.Write Mass=true
    Module.Write Material Random=true
    Module.Write Position=false
    Module.Write RibbonTwist=false
    Module.Write RibbonWidth=true
    Ribbon Width=45
   )
  ]
  particleUpdateScript [
   &ParticleState(Kill Particles When Lifetime Has Elapsed=false, Let Infinitely Lived Particles Die When Emitter Deactivates=true)
   &SkeletalMeshLocation(
    Apply To Particle Position=false
    Inherit Velocity=false
    MaskSpawnGroup=false
    Mesh Sampling Type=::NewEnumerator5
    Sampled Position Offset Coordinate Space:ENiagaraMeshTransforms=::NewEnumerator3
    Skeletal Mesh=$User.Skeletal Mesh
    Socket Sampling Mode=::NewEnumerator5
   )
   /SetVariables(Particles.OnFrontOfGunPos:Vector3f=&TransformPosition(Position=<150 0 0>, Source Space=::Local, Destination Space=::World)) {enabled=false, isSetParameters=true}
   &FN_HermiteCubicSpline(
    Position Blend:NiagaraFloat=$Particles.Initial.Partition.NormalizedID
    Position End:Vector3f=$Output.SkeletalMeshLocation.SampledPosition
    Position Start:Vector3f=&MakeCustomVectorFromBool(
     Bool=$User.bIsHitting
     True Vector=$User.ImpactLocation
     False Vector=&TransformPosition(Position=<150 0 0>, Source Space=::Local, Destination Space=::World)
    )
    Write Particles Position=true
   ) {enabled=false}
   &FN_HermiteCubicSpline(
    Direction End:Vector3f=&MakeCustomVectorFromBool(
     Bool=$User.bIsHitting
     True Vector=&Multiply_VectorByFloat(
      Vector=&RotateVectorByQuaternion(Vector To Rotate=<1 0 0>, Quaternion=$Output.SkeletalMeshLocation.SampledRotation)
      Float=$System.TangentLength
     )
     False Vector=<1 0 0>
    )
    Direction Start:Vector3f=&MakeCustomVectorFromBool(
     Bool=$User.bIsHitting
     True Vector=&Multiply_VectorByFloat(Vector=$User.Impact_Orientation, Float=$System.TangentLength)
     False Vector=<1 0 0>
    )
    Position Blend:NiagaraFloat=$Particles.Initial.Partition.NormalizedID
    Position End:Vector3f=&Add_Vector(
     A=$Output.SkeletalMeshLocation.SampledPosition
     B=&RotateVectorByQuaternion(Vector To Rotate=<-12.5 0 0>, Quaternion=$Output.SkeletalMeshLocation.SampledRotation)
    )
    Position Start:Vector3f=&self:SmoothInterpolateParticleVector(
     Initial Vector:Vector3f=$System.PosOnFrontOfGun
     Vector to Interpolate:Vector3f=&MakeCustomVectorFromBool(Bool=$User.bIsHitting, True Vector=$User.ImpactLocation, False Vector=$System.PosOnFrontOfGun)
     Rate Of Change:NiagaraFloat=$System.PositionInterpolateSpeed
    )
    Write Particles Position=true
   ) {module=FN_HermiteCubicSpline001}
   &ScaleRibbonWidth(Ribbon Width Scale=&FloatFromCurve(
     FloatCurve=DataInterface() {propertyValues {Curve {keys [<at 0 1 RCIM_Cubic>, {interpMode=RCIM_Cubic, leaveTangent=0.4, time=1, value=0.025}]}}}
     CurveIndex=$Particles.Partition.NormalizedID
     Scale Curve=5
    ))
   &DynamicMaterialParameters(
    Index 0 Param 2=&FloatFromCurve(
     FloatCurve=DataInterface() {propertyValues {Curve {keys [{arriveTangent=-0.8726, interpMode=RCIM_Cubic, leaveTangent=-0.8726, tangentMode=RCTM_User, time=0, value=1}, <at 1 0 RCIM_Cubic>]}}}
     CurveIndex=$Particles.Initial.Partition.NormalizedID
     Scale Curve=1.7
    )
    Module.Param0WriteEnabled=true
    Module.Param1WriteEnabled=true
    Module.Param2WriteEnabled=true
    Module.Param3WriteEnabled=true
   )
   &ScaleColor(
    Scale Alpha=&FloatFromCurve(
     FloatCurve=DataInterface() {propertyValues {Curve {keys [<at 0 0 RCIM_Cubic>,<at 0.9 1>,<at 1 0 RCIM_Cubic>]}}}
     CurveIndex=$Particles.Initial.Partition.NormalizedID
     Scale Curve=1.5
    )
    ScaleA=true
    ScaleRGB=true
   )
  ]
  rendererClasses [&NiagaraRibbonRendererProperties]
  renderers [{rendererClass=&NiagaraRibbonRendererProperties}]
 }
 Water_Foam {
  emitterUpdateScript [
   &EmitterState()
   &SpawnRate(SpawnRate=&Lerp_Float(A=0, B=6, Alpha=$User.Underwater), Use Spawn Probability=false)
  ]
  particleSpawnScript [
   &InitializeParticle(
    Lifetime=&RandomRangeFloat(Minimum=1.5, Maximum=2.5)
    Mesh Scale=&VectorFromFloat(Value=100)
    Module.Write Color=true
    Module.Write Lifetime=true
    Module.Write Mass=true
    Module.Write Material Random=true
    Module.Write Position=false
    Module.Write Scale=true
    Module.Write SpriteRotation=false
    Module.Write SpriteSize=true
    Sprite Size=&Vector2DFromFloat(Value=100)
   )
   /Niagara/Modules/Spawn/Orientation/V3/InitialMeshOrientation(
    Rotation=&RandomRangeVector(Minimum=<0 0 -1>, Maximum=<0 0 1>)
    Use Orientation Vector=true
    Use Rotation Vector=true
   ) {module=InitialMeshOrientation001}
   &DynamicMaterialParameters(
    Index 0 Param 1=$User.WaterBodyIndex
    Index 0 Param 2=$User.WaterDepth
    Index 1 Param 3=200
    Index 3 Param 1=$User.WaterBodyIndex
    Index 3 Param 2=$User.WaterDepth
    Module.Param0WriteEnabled=true
    Module.Param0WriteEnabled1=false
    Module.Param0WriteEnabled3=true
    Module.Param1WriteEnabled=true
    Module.Param1WriteEnabled1=false
    Module.Param1WriteEnabled3=true
    Module.Param2WriteEnabled=true
    Module.Param2WriteEnabled1=true
    Module.Param2WriteEnabled3=true
    Module.Param3WriteEnabled=true
    Module.Param3WriteEnabled1=false
    Module.Param3WriteEnabled3=true
    Write Parameter Index 1=true
    Write Parameter Index 3=true
   ) {module=DynamicMaterialParameters003}
   &SkeletalMeshLocation(
    Inherit Velocity=false
    MaskSpawnGroup=false
    Mesh Sampling Type=::NewEnumerator5
    Sampled Position Offset=<100 0 0>
    Sampled Position Offset Coordinate Space:ENiagaraMeshTransforms=::NewEnumerator3
    Skeletal Mesh=$User.Skeletal Mesh
    Socket Sampling Mode=::NewEnumerator5
   )
   &CylinderLocation(
    Cylinder Height=0
    Cylinder Radius=50
    MaskSpawnGroup=false
    Module.Enable Lathe Profile=false
    Module.Override Randomness=false
    Module.Surface Only=false
   )
   &InitializeWaterSnapToOwner()
  ]
  particleUpdateScript [
   &ParticleState()
   &DynamicMaterialParameters(
    Index 2 Param 1=&FloatFromCurve(FloatCurve=DataInterface() {propertyValues {Curve {keys [<at -0.0046 0.25>,<at 1.0011 1.5>]}}})
    Index 2 Param 2=&FloatFromCurve(FloatCurve=DataInterface() {propertyValues {Curve {keys [<at -0.0046 0.1>,<at 0.25 1.5>,<at 1.0011 0.3>]}}})
    Index 2 Param 3=1
    Index 2 Param 4=&FloatFromCurve(FloatCurve=DataInterface() {propertyValues {Curve {keys [
         {arriveTangent=-1.9991, interpMode=RCIM_Cubic, time=0, value=1}
         {arriveTangent=-1.545, interpMode=RCIM_Cubic, leaveTangent=-1.545, tangentMode=RCTM_User, time=0.2095, value=0.6901}
         {arriveTangent=-0.5152, interpMode=RCIM_Cubic, leaveTangent=-0.5152, tangentMode=RCTM_User, time=1, value=0}
        ]}}})
    Module.Param0WriteEnabled2=true
    Module.Param1WriteEnabled2=true
    Module.Param2WriteEnabled2=true
    Module.Param3WriteEnabled2=true
    Write Parameter Index 0=false
    Write Parameter Index 2=true
   ) {module=DynamicMaterialParameters002}
   &ScaleColor(
    Scale Alpha=&FloatFromCurve(FloatCurve=DataInterface() {propertyValues {Curve {keys [<at 0 0 RCIM_Cubic>,<at 0.5 1 RCIM_Cubic>,<at 1 0 RCIM_Cubic>]}}})
    ScaleA=true
    ScaleRGB=true
   ) {enabled=false}
   &ScaleSpriteSize(Scale Factor=&Vector2DFromFloat(Value=&FloatFromCurve(FloatCurve=DataInterface() {propertyValues {Curve {keys [<at 0 0>,<at 1 1>]}}}))) {enabled=false}
   &ConstrainPositionToPlane(Plane Normal:Vector3f=$Particles.WaterPlaneNormal, Plane Position:Vector3f=$Particles.WaterPlaneLocation)
   &CameraOffset(Camera Offset Amount=10)
   &ScaleMeshSize(Scale Factor=&VectorFromFloat(Value=&FloatFromCurve(FloatCurve=DataInterface() {propertyValues {Curve {keys [<at 0 0.25>,<at 1 1.5>]}}})))
  ]
  rendererClasses [&NiagaraSpriteRendererProperties, &NiagaraMeshRendererProperties]
  renderers [
   {rendererClass=&NiagaraSpriteRendererProperties}
   {rendererClass=&NiagaraMeshRendererProperties, rendererIndex=1}
  ]
 }
}
summary {
 systemName=NS_RocketLauncher_JunkGun_Suction
 emitters [
  {emitterName=TubeRibbon, rendererClasses [&NiagaraRibbonRendererProperties]}
  {emitterName=SpeedLines, rendererClasses [&NiagaraSpriteRendererProperties]}
  {emitterName=Smoke, rendererClasses [&NiagaraSpriteRendererProperties]}
  {emitterName=DebrisSprite, rendererClasses [&NiagaraSpriteRendererProperties]}
  {emitterName=Debris3D001, rendererClasses [&NiagaraMeshRendererProperties]}
  {
   emitterName=Drops_SnapToSurface_Detail
   rendererClasses [&NiagaraSpriteRendererProperties, &NiagaraMeshRendererProperties]
  }
  {emitterName=Bubbles_Muzzle, rendererClasses [&NiagaraMeshRendererProperties]}
  {emitterName=Water_Foam, rendererClasses [&NiagaraSpriteRendererProperties, &NiagaraMeshRendererProperties]}
 ]
 userVariables [
  User.AlphaMult {type=NiagaraFloat, defaultValue=1, description=""}
  User.bIsCollecting {type=NiagaraBool}
  User.bIsHitting {type=NiagaraBool, defaultValue=true}
  User.CollectingMaterial {type=NiagaraFloat, defaultValue=1, description=""}
  User.CorrectDirection {type=NiagaraFloat, defaultValue=0, description=""}
  User.FogTop {type=Vector3f, description="", defaultValue {struct=Vector3f, value=<0 0 0>}}
  User.Impact_Orientation {type=Vector3f, description="", defaultValue {struct=Vector3f, value=<1 0 0>}}
  User.ImpactLocation {type=Vector3f, description="", defaultValue {struct=Vector3f, value=<150 0 0>}}
  User.RibbonIDOffset {type=NiagaraInt32}
  User.Skeletal Mesh {
   type=NiagaraDataInterfaceSkeletalMesh
   defaultValue {
    struct=NiagaraExt_VariableValue_DataInterface
    value {dataInterface=&self:NiagaraDataInterfaceSkeletalMesh_2}
   }
  }
  User.SplineUpdatedEndPos {type=Vector3f, description="", defaultValue {struct=Vector3f, value=<0 0 0>}}
  User.Underwater {type=NiagaraFloat, defaultValue=0, description=""}
  User.WaterBodyIndex {type=NiagaraFloat, defaultValue=0, description=""}
  User.WaterDepth {type=NiagaraFloat, defaultValue=0, description=""}
  User.WaterPlaneLocation {type=Vector3f, description="", defaultValue {struct=Vector3f, value=<0 0 0>}}
  User.WaterPlaneNormal {type=Vector3f, description="", defaultValue {struct=Vector3f, value=<0 0 1>}}
  User.WaveAttenuationFactor {type=NiagaraFloat, defaultValue=0, description=""}
  User.WaveOffset {type=Vector3f, description="", defaultValue {struct=Vector3f, value=<0 0 0>}}
  User.WaveTime {type=NiagaraFloat, defaultValue=0, description=""}
 ]
}
