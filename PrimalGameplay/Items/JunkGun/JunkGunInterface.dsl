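; JunkGunInterface is a Blueprint interface for the junk gun item, letting other code query its suction state.
; It declares GetJunkGunSuctionValues, which returns two booleans, two vectors and a float.
; The function body returns only default false and zero values, with no logic of its own.
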(fn GetJunkGunSuctionValues
  (return false "0, 0, 0" false 0.0 "0, 0, 0"))
